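# Physics asset for the back flap on the F_MED_Monarch female player body, for simulating the flap's motion against the body.
# A chain of five capsules, dyn_backFlap to dyn_backFlap_04, hangs from a small dyn_backFlap_MSTR body constrained to the pelvis.
# Capsules on the pelvis, upper spine and thighs supply body volumes, linked by free angular constraints.
# Flap links are hinge-like, with tight swing1 and twist limits and a wide swing2 range.
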PhysicsAsset /BRCosmetics/Characters/Player/Female/Medium/Bodies/F_MED_Monarch/Meshes/F_MED_Monarch_BackFlap_Physics
bodies {
 dyn_backFlap [{shapeType=Capsule, length=11.4814, radius=2.6, shapeName=dyn_backFlap_capsule, center=<3.4417 0 2>, rotation=<rot 0 0 90>}]
 dyn_backFlap_01 [{shapeType=Capsule, length=10.4974, radius=3, shapeName=dyn_backFlap_01_capsule, center=<2.7014 0 2>, rotation=<rot 0 0 90>}]
 dyn_backFlap_02 [{shapeType=Capsule, length=10.4974, radius=4.1222, shapeName=dyn_backFlap_02_capsule, center=<2.7014 0 2>, rotation=<rot 0 0 90>}]
 dyn_backFlap_03 [{shapeType=Capsule, length=11.1231, radius=3.4965, shapeName=dyn_backFlap_03_capsule, center=<2.7014 0 2>, rotation=<rot 0 0 90>}]
 dyn_backFlap_04 [{shapeType=Capsule, length=11.6527, radius=2.9668, shapeName=dyn_backFlap_04_capsule, center=<2.7014 0 2>, rotation=<rot 0 0 90>}]
 dyn_backFlap_MSTR [{shapeType=Capsule, length=0.505, radius=0.505, shapeName=dyn_backFlap_MSTR_capsule, rotation=<rot 0 -90 0>}]
 pelvis [{shapeType=Capsule, length=79.8487, radius=12.0389, shapeName=pelvis_capsule, center=<-2.5557 1.0837 0>, rotation=<rot 0 22.654 0>}]
 spine_02 [{shapeType=Capsule, length=39.5333, radius=9.3, shapeName=spine_02_capsule, center=<0 0.3182 0>}]
 spine_03 [{shapeType=Capsule, length=3.0205, radius=9.6241, shapeName=spine_03_capsule}]
 spine_04 [{shapeType=Capsule, length=6.0967, radius=9.6241, shapeName=spine_04_capsule}]
 thigh_l [{shapeType=Capsule, length=27.2756, radius=8.738, shapeName=thigh_l_capsule, center=<-13.9049 -4.3679 0.6459>, rotation=<rot 90 0 0>}]
 thigh_r [{shapeType=Capsule, length=30.2633, radius=7.7242, shapeName=thigh_r_capsule, center=<18.8055 5.5878 0>, rotation=<rot 90 0 0>}]
}
constraints [
 {
  bone1Name=spine_02
  bone2Name=pelvis
  swing1LimitDegrees=45
  swing1Motion=Free
  swing2LimitDegrees=45
  swing2Motion=Free
  twistLimitDegrees=45
  twistMotion=Free
 }
 {
  bone1Name=spine_04
  bone2Name=spine_03
  swing1LimitDegrees=45
  swing1Motion=Free
  swing2LimitDegrees=45
  swing2Motion=Free
  twistLimitDegrees=45
  twistMotion=Free
 }
 {
  bone1Name=spine_03
  bone2Name=spine_02
  swing1LimitDegrees=45
  swing1Motion=Free
  swing2LimitDegrees=45
  swing2Motion=Free
  twistLimitDegrees=45
  twistMotion=Free
 }
 {
  bone1Name=thigh_r
  bone2Name=pelvis
  swing1LimitDegrees=45
  swing1Motion=Free
  swing2LimitDegrees=45
  swing2Motion=Free
  twistLimitDegrees=45
  twistMotion=Free
 }
 {
  bone1Name=thigh_l
  bone2Name=pelvis
  swing1LimitDegrees=45
  swing1Motion=Free
  swing2LimitDegrees=45
  swing2Motion=Free
  twistLimitDegrees=45
  twistMotion=Free
 }
 {
  bone1Name=dyn_backFlap_MSTR
  bone2Name=pelvis
  swing1LimitDegrees=45
  swing1Motion=Free
  swing2LimitDegrees=45
  swing2Motion=Free
  twistLimitDegrees=45
  twistMotion=Free
 }
 {bone1Name=dyn_backFlap, bone2Name=dyn_backFlap_MSTR, swing1LimitDegrees=1, swing2LimitDegrees=45, twistLimitDegrees=1}
 {bone1Name=dyn_backFlap_01, bone2Name=dyn_backFlap, swing1LimitDegrees=1, swing2LimitDegrees=45, twistLimitDegrees=1}
 {bone1Name=dyn_backFlap_02, bone2Name=dyn_backFlap_01, swing1LimitDegrees=1, swing2LimitDegrees=45, twistLimitDegrees=1}
 {bone1Name=dyn_backFlap_03, bone2Name=dyn_backFlap_02, swing1LimitDegrees=1, swing2LimitDegrees=45, twistLimitDegrees=1}
 {bone1Name=dyn_backFlap_04, bone2Name=dyn_backFlap_03, swing1LimitDegrees=1, swing2LimitDegrees=45, twistLimitDegrees=1}
]
