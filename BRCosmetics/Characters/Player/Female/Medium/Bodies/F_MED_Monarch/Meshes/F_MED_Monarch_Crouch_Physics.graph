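# Physics asset for the crouch variant of the F_MED_Monarch female medium body, covering its dangling front flap and rope.
# Capsules on the pelvis, spine and thighs form body volumes, joined by free angular constraints.
# The flap and rope are four-box chains, each rooted in a tiny MSTR capsule that hangs off the pelvis with free rotation.
# Chain links have angular limits, with the flap allowing wider swing than the rope.
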PhysicsAsset /BRCosmetics/Characters/Player/Female/Medium/Bodies/F_MED_Monarch/Meshes/F_MED_Monarch_Crouch_Physics
bodies {
 dyn_frontFlap_01 [{shapeType=Box, extentX=8.3274, extentY=43.784, extentZ=0.5, shapeName=dyn_frontFlap_01_box, center=<4.3983 0 1>}]
 dyn_frontFlap_02 [{shapeType=Box, extentX=8.5387, extentY=33.435, extentZ=1.022, shapeName=dyn_frontFlap_02_box, center=<3.9391 0 0>}]
 dyn_frontFlap_03 [{shapeType=Box, extentX=8.874, extentY=24.4445, extentZ=1.022, shapeName=dyn_frontFlap_03_box, center=<4.261 0 0>}]
 dyn_frontFlap_04 [{shapeType=Box, extentX=8.1688, extentY=16.0525, extentZ=1.022, shapeName=dyn_frontFlap_04_box, center=<4.3308 0 0>}]
 dyn_frontFlap_MSTR [{shapeType=Capsule, length=0.505, radius=0.505, shapeName=dyn_frontFlap_MSTR_capsule}]
 dyn_rope_01 [{shapeType=Box, extentX=4.7024, extentY=14.8748, extentZ=2.2397, shapeName=dyn_rope_01_box, center=<6.9012 -0.0001 -0.0001>}]
 dyn_rope_02 [{shapeType=Box, extentX=9.0178, extentY=19.2927, extentZ=2.3114, shapeName=dyn_rope_02_box, center=<5.0751 -0.0001 -0.0001>}]
 dyn_rope_03 [{shapeType=Box, extentX=9.0178, extentY=19.0641, extentZ=2.3114, shapeName=dyn_rope_03_box, center=<5.0751 -0.0001 -0.0001>}]
 dyn_rope_04 [{shapeType=Box, extentX=8.5116, extentY=21.0235, extentZ=0.5, shapeName=dyn_rope_04_box, center=<4.6381 -0.0001 -0.0001>}]
 dyn_rope_MSTR [{shapeType=Capsule, length=0.505, radius=0.505, shapeName=dyn_rope_MSTR_capsule}]
 pelvis [
  {shapeType=Capsule, length=7.2709, radius=10.4375, shapeName=pelvis_capsule, center=<0 -1.6232 0>}
  {shapeType=Capsule, length=7.6556, radius=10.0528, shapeName=pelvis_capsule, center=<-3.3225 -1.1565 0.0003>}
 ]
 spine_01 [{shapeType=Capsule, length=5.2223, radius=9.6241, shapeName=spine_01_capsule}]
 spine_02 [{shapeType=Capsule, length=3.8791, radius=9.6241, shapeName=spine_02_capsule}]
 spine_03 [{shapeType=Capsule, length=3.0205, radius=9.6241, shapeName=spine_03_capsule}]
 spine_04 [{shapeType=Capsule, length=6.0967, radius=9.6241, shapeName=spine_04_capsule}]
 thigh_l [{shapeType=Capsule, length=27.8774, radius=8.1362, shapeName=thigh_l_capsule, center=<-19.2678 0 -0.0001>, rotation=<rot 90 0 0>}]
 thigh_r [{shapeType=Capsule, length=29.5828, radius=8.4046, shapeName=thigh_r_capsule, center=<18.8055 0.0002 0>, rotation=<rot 90 0 0>}]
}
constraints [
 {
  bone1Name=spine_01
  bone2Name=pelvis
  swing1LimitDegrees=45
  swing1Motion=Free
  swing2LimitDegrees=45
  swing2Motion=Free
  twistLimitDegrees=45
  twistMotion=Free
 }
 {
  bone1Name=spine_02
  bone2Name=spine_01
  swing1LimitDegrees=45
  swing1Motion=Free
  swing2LimitDegrees=45
  swing2Motion=Free
  twistLimitDegrees=45
  twistMotion=Free
 }
 {
  bone1Name=spine_04
  bone2Name=spine_03
  swing1LimitDegrees=45
  swing1Motion=Free
  swing2LimitDegrees=45
  swing2Motion=Free
  twistLimitDegrees=45
  twistMotion=Free
 }
 {
  bone1Name=spine_03
  bone2Name=spine_02
  swing1LimitDegrees=45
  swing1Motion=Free
  swing2LimitDegrees=45
  swing2Motion=Free
  twistLimitDegrees=45
  twistMotion=Free
 }
 {
  bone1Name=thigh_r
  bone2Name=pelvis
  swing1LimitDegrees=45
  swing1Motion=Free
  swing2LimitDegrees=45
  swing2Motion=Free
  twistLimitDegrees=45
  twistMotion=Free
 }
 {bone1Name=dyn_rope_01, bone2Name=dyn_rope_MSTR, swing1LimitDegrees=25, swing2LimitDegrees=25, twistLimitDegrees=30}
 {bone1Name=dyn_rope_02, bone2Name=dyn_rope_01, swing1LimitDegrees=25, swing2LimitDegrees=25, twistLimitDegrees=30}
 {bone1Name=dyn_rope_03, bone2Name=dyn_rope_02, swing1LimitDegrees=25, swing2LimitDegrees=25, twistLimitDegrees=30}
 {bone1Name=dyn_rope_04, bone2Name=dyn_rope_03, swing1LimitDegrees=25, swing2LimitDegrees=25, twistLimitDegrees=30}
 {
  bone1Name=thigh_l
  bone2Name=pelvis
  swing1LimitDegrees=45
  swing1Motion=Free
  swing2LimitDegrees=45
  swing2Motion=Free
  twistLimitDegrees=45
  twistMotion=Free
 }
 {bone1Name=dyn_frontFlap_01, bone2Name=dyn_frontFlap_MSTR, swing1LimitDegrees=30, swing2LimitDegrees=60, twistLimitDegrees=30}
 {bone1Name=dyn_frontFlap_02, bone2Name=dyn_frontFlap_01, swing1LimitDegrees=30, swing2LimitDegrees=60, twistLimitDegrees=30}
 {bone1Name=dyn_frontFlap_03, bone2Name=dyn_frontFlap_02, swing1LimitDegrees=30, swing2LimitDegrees=60, twistLimitDegrees=30}
 {bone1Name=dyn_frontFlap_04, bone2Name=dyn_frontFlap_03, swing1LimitDegrees=30, swing2LimitDegrees=60, twistLimitDegrees=30}
 {
  bone1Name=dyn_rope_MSTR
  bone2Name=pelvis
  swing1LimitDegrees=45
  swing1Motion=Free
  swing2LimitDegrees=45
  swing2Motion=Free
  twistLimitDegrees=45
  twistMotion=Free
 }
 {
  bone1Name=dyn_frontFlap_MSTR
  bone2Name=pelvis
  swing1LimitDegrees=45
  swing1Motion=Free
  swing2LimitDegrees=45
  swing2Motion=Free
  twistLimitDegrees=45
  twistMotion=Free
 }
]
