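# Physics asset for the Monarch female medium player body cosmetic, setting up secondary motion for its dangling front flap and rope.
# Both are four-segment chains of capsules and boxes with limited swing and twist, each rooted at a small master body constrained to the pelvis; the flap swings wider than the rope.
# Capsules on the pelvis, spine, thighs and right hand provide body shapes, with free-rotation constraints linking all but the unconstrained hand.
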PhysicsAsset /BRCosmetics/Characters/Player/Female/Medium/Bodies/F_MED_Monarch/Meshes/F_MED_Monarch_Physics
bodies {
 dyn_frontFlap_01 [{shapeType=Capsule, length=9.5093, radius=1.6979, shapeName=dyn_frontFlap_01_capsule, center=<3.724 0 0.9622>, rotation=<rot 0 0 90>}]
 dyn_frontFlap_02 [
  {shapeType=Capsule, length=11.473, radius=1.5746, shapeName=dyn_frontFlap_02_capsule, center=<3.724 0 0.912>, rotation=<rot 0 0 90>}
  {shapeType=Capsule, length=2.2093, radius=1.6346, shapeName=dyn_frontFlap_02_capsule, center=<1.6614 8.1405 0.912>, rotation=<rot 0 0 90>}
  {shapeType=Capsule, length=2.7683, radius=1.0756, shapeName=dyn_frontFlap_02_capsule, center=<1.7839 -7.1149 0.912>, rotation=<rot 0 0 90>}
 ]
 dyn_frontFlap_03 [{shapeType=Capsule, length=15.2058, radius=2.287, shapeName=dyn_frontFlap_03_capsule, center=<1.2741 0 0.9162>, rotation=<rot 0 0 90>}]
 dyn_frontFlap_04 [{shapeType=Capsule, length=7.8112, radius=1.8191, shapeName=dyn_frontFlap_04_capsule, center=<1.7051 0 0.9423>, rotation=<rot 0 0 90>}]
 dyn_frontFlap_MSTR [{shapeType=Capsule, length=0.505, radius=0.505, shapeName=dyn_frontFlap_MSTR_capsule}]
 dyn_rope_01 [{shapeType=Box, extentX=4.7024, extentY=14.8748, extentZ=1.9605, shapeName=dyn_rope_01_box, center=<6.9012 -0.0001 0.4026>}]
 dyn_rope_02 [{shapeType=Capsule, length=13, radius=2.5, shapeName=dyn_rope_02_capsule, center=<5.3928 0 0.4339>, rotation=<rot 0 0 90>}]
 dyn_rope_03 [{shapeType=Capsule, length=17.6611, radius=2, shapeName=dyn_rope_03_capsule, center=<4.0901 0 0.0299>, rotation=<rot 0 0 90>}]
 dyn_rope_04 [{shapeType=Capsule, length=13, radius=2, shapeName=dyn_rope_04_capsule, center=<1.489 0 1.6387>, rotation=<rot 0 0 90>}]
 dyn_rope_MSTR [{shapeType=Capsule, length=0.505, radius=0.505, shapeName=dyn_rope_MSTR_capsule}]
 hand_r [{shapeType=Capsule, length=25, radius=4, shapeName=hand_r_capsule, center=<-5.4162 -1.3109 0>, rotation=<rot -90 179.9802 -179.9803>}]
 pelvis [
  {shapeType=Capsule, length=9.8696, radius=9.0907, shapeName=pelvis_capsule, center=<-0.8652 -1.4553 0>}
  {shapeType=Capsule, length=7.832, radius=9.8764, shapeName=pelvis_capsule, center=<-4.2158 -1.8174 0.0003>}
 ]
 spine_01 [{shapeType=Capsule, length=5.2223, radius=9.6241, shapeName=spine_01_capsule}]
 spine_02 [{shapeType=Capsule, length=3.8791, radius=9.6241, shapeName=spine_02_capsule}]
 spine_03 [{shapeType=Capsule, length=3.0205, radius=9.6241, shapeName=spine_03_capsule}]
 spine_04 [{shapeType=Capsule, length=6.0967, radius=9.6241, shapeName=spine_04_capsule}]
 thigh_l [
  {shapeType=Capsule, length=27.8774, radius=8.1362, shapeName=thigh_l_capsule, center=<-19.2678 0 -0.0001>, rotation=<rot 90 0 0>}
  {shapeType=Capsule, length=27.8774, radius=8.1362, shapeName=thigh_l_capsule, center=<-19.2582 -4.7249 -0.0003>, rotation=<rot 90 0 0>}
 ]
 thigh_r [
  {shapeType=Capsule, length=30.2633, radius=7.7242, shapeName=thigh_r_capsule, center=<18.8055 0.0002 0>, rotation=<rot 90 0 0>}
  {shapeType=Capsule, length=30.2633, radius=7.7242, shapeName=thigh_r_capsule, center=<18.271 5.2503 0.0001>, rotation=<rot 90 0 0>}
 ]
}
constraints [
 {
  bone1Name=spine_01
  bone2Name=pelvis
  swing1LimitDegrees=45
  swing1Motion=Free
  swing2LimitDegrees=45
  swing2Motion=Free
  twistLimitDegrees=45
  twistMotion=Free
 }
 {
  bone1Name=spine_02
  bone2Name=spine_01
  swing1LimitDegrees=45
  swing1Motion=Free
  swing2LimitDegrees=45
  swing2Motion=Free
  twistLimitDegrees=45
  twistMotion=Free
 }
 {
  bone1Name=spine_04
  bone2Name=spine_03
  swing1LimitDegrees=45
  swing1Motion=Free
  swing2LimitDegrees=45
  swing2Motion=Free
  twistLimitDegrees=45
  twistMotion=Free
 }
 {
  bone1Name=spine_03
  bone2Name=spine_02
  swing1LimitDegrees=45
  swing1Motion=Free
  swing2LimitDegrees=45
  swing2Motion=Free
  twistLimitDegrees=45
  twistMotion=Free
 }
 {
  bone1Name=thigh_r
  bone2Name=pelvis
  swing1LimitDegrees=45
  swing1Motion=Free
  swing2LimitDegrees=45
  swing2Motion=Free
  twistLimitDegrees=45
  twistMotion=Free
 }
 {bone1Name=dyn_rope_01, bone2Name=dyn_rope_MSTR, swing1LimitDegrees=15, swing2LimitDegrees=15, twistLimitDegrees=10}
 {bone1Name=dyn_rope_02, bone2Name=dyn_rope_01, swing1LimitDegrees=10, swing2LimitDegrees=15, twistLimitDegrees=5}
 {bone1Name=dyn_rope_03, bone2Name=dyn_rope_02, swing1LimitDegrees=10, swing2LimitDegrees=15, twistLimitDegrees=5}
 {bone1Name=dyn_rope_04, bone2Name=dyn_rope_03, swing1LimitDegrees=10, swing2LimitDegrees=15, twistLimitDegrees=5}
 {
  bone1Name=thigh_l
  bone2Name=pelvis
  swing1LimitDegrees=45
  swing1Motion=Free
  swing2LimitDegrees=45
  swing2Motion=Free
  twistLimitDegrees=45
  twistMotion=Free
 }
 {bone1Name=dyn_frontFlap_01, bone2Name=dyn_frontFlap_MSTR, swing1LimitDegrees=15, swing2LimitDegrees=60, twistLimitDegrees=15}
 {bone1Name=dyn_frontFlap_02, bone2Name=dyn_frontFlap_01, swing1LimitDegrees=15, swing2LimitDegrees=60, twistLimitDegrees=15}
 {bone1Name=dyn_frontFlap_03, bone2Name=dyn_frontFlap_02, swing1LimitDegrees=15, swing2LimitDegrees=60, twistLimitDegrees=15}
 {bone1Name=dyn_frontFlap_04, bone2Name=dyn_frontFlap_03, swing1LimitDegrees=15, swing2LimitDegrees=60, twistLimitDegrees=15}
 {
  bone1Name=dyn_frontFlap_MSTR
  bone2Name=pelvis
  swing1LimitDegrees=45
  swing1Motion=Free
  swing2LimitDegrees=45
  swing2Motion=Free
  twistLimitDegrees=45
  twistMotion=Free
 }
 {
  bone1Name=dyn_rope_MSTR
  bone2Name=pelvis
  swing1LimitDegrees=45
  swing1Motion=Free
  swing2LimitDegrees=45
  swing2Motion=Free
  twistLimitDegrees=45
  twistMotion=Free
 }
]
