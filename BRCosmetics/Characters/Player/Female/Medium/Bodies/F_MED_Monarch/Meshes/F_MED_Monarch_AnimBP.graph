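# Post-process style AnimBP for the F_MED_Monarch character body, animating its rope and front and back flaps.
# The copied parent pose runs through a Control Rig, crouch-switched rope RigidBody sims, then flap RigidBody and AnimDynamics.
# Bone offsets and constraints correct for movement states and long back blings, and pin the rope as the right hand nears it.
# The cached result gets extra rope constraints when rigid bodies are disabled, then final rope adjustments.
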AnimBlueprint /BRCosmetics/Characters/Player/Female/Medium/Bodies/F_MED_Monarch/Meshes/F_MED_Monarch_AnimBP
target_skeleton=&/F_MED_Monarch_Skeleton
graphs [
 AnimGraph {nodes [
   n1 = "Constraint - dyn_rope_MSTR (R)"(ComponentPose=Local To Component(LocalPose=Use cached pose 'PostSim'()), Alpha=0.5, ConstraintWeights_0=1) {
    alphaBoolBlend {blendOption=Linear}
    boneToModify {boneName=dyn_rope_MSTR}
    constraintSetup [{transformType=Rotation, targetBone {boneName=thigh_twist_01_r}}]
    constraintWeights [1]
   }
   n2 = "Transform (Modify) Bone - Bone: dyn_rope_MSTR"!(ComponentPose=@n1, bAlphaBoolEnabled:bool=True, Translation="0,0,0", Rotation="5,-15,0", Scale="1,1,1") -> (Pose) {
    alphaInputType=Bool
    rotationMode=BMM_Additive
    scaleMode=BMM_Additive
    translationMode=BMM_Additive
    alphaBoolBlend {blendOption=Linear}
    boneToModify {boneName=dyn_rope_MSTR}
    rotation {pitch=5, yaw=-15}
   }
   n3 = "Constraint - dyn_rope_02 (PPPPPP)"!(
    ComponentPose=@n2
    Alpha=1
    ConstraintWeights_0=0.0145
    ConstraintWeights_1:float=0.022
    ConstraintWeights_2:float=0.0058
    ConstraintWeights_3:float=0.1994
    ConstraintWeights_4:float=0.1213
    ConstraintWeights_5:float=0.6369
   ) -> (Pose) {
    alphaBoolBlend {blendOption=Linear}
    alphaScaleBiasClamp {bInterpResult=true, interpSpeedIncreasing=20}
    boneToModify {boneName=dyn_rope_02}
    constraintSetup [
     {targetBone {boneName=spine_01}}
     {targetBone {boneName=spine_02}}
     {targetBone {boneName=spine_04}}
     {targetBone {boneName=pelvis}}
     {targetBone {boneName=thigh_r}}
     {targetBone {boneName=thigh_twist_01_r}}
    ]
    constraintWeights [1,1,1,1,1,1]
   }
   n4 = "Constraint - dyn_rope_03 (PPPPPPP)"!(
    ComponentPose=@n3
    Alpha=1
    ConstraintWeights_0=0
    ConstraintWeights_1:float=0.0007
    ConstraintWeights_2:float=0.001
    ConstraintWeights_3:float=0.0002
    ConstraintWeights_4:float=0.0425
    ConstraintWeights_5:float=0.0001
    ConstraintWeights_6:float=0.9556
   ) -> (Pose) {
    alphaBoolBlend {blendOption=Linear}
    alphaScaleBiasClamp {bInterpResult=true, interpSpeedIncreasing=20}
    boneToModify {boneName=dyn_rope_03}
    constraintSetup [
     {targetBone {boneName=calf_r}}
     {targetBone {boneName=spine_01}}
     {targetBone {boneName=spine_02}}
     {targetBone {boneName=spine_04}}
     {targetBone {boneName=pelvis}}
     {targetBone {boneName=thigh_r}}
     {targetBone {boneName=thigh_twist_01_r}}
    ]
    constraintWeights [1,1,1,1,1,1,1]
   }
   n5 = "Transform (Modify) Bone - Bone: dyn_rope_02"(ComponentPose=@n4, Alpha=1, Translation="0,0,-1", Rotation="-10,0,0", Scale="1,1,1") {
    rotationMode=BMM_Additive
    rotationSpace=BCS_BoneSpace
    translationMode=BMM_Additive
    translationSpace=BCS_BoneSpace
    alphaBoolBlend {blendOption=Linear}
    alphaScaleBiasClamp {interpSpeedIncreasing=30}
    boneToModify {boneName=dyn_rope_02}
    rotation {pitch=-10}
    translation {z=-1}
   }
   n6 = Component To Local(ComponentPose=@n5)
   n7 = Use cached pose 'PostSim'()
   n8 = Get bRigidBodyEnabled()
   n9 = "Transform (Modify) Bone - Bone: dyn_rope_MSTR"!(
    ComponentPose=Local To Component(LocalPose=Blend Poses by bool(BlendPose_0=@n7, BlendPose_1=@n6, bActiveValue=@n8) {blendType=Linear})
    Alpha:float=1
    Translation="0,0,-0.5"
    Rotation="5,-45,-10"
    Scale="1,1,1"
   ) -> (Pose) {
    rotationMode=BMM_Additive
    rotationSpace=BCS_BoneSpace
    translationMode=BMM_Additive
    translationSpace=BCS_BoneSpace
    alphaBoolBlend {blendOption=Linear}
    alphaScaleBiasClamp {bInterpResult=true, interpSpeedDecreasing=5, interpSpeedIncreasing=1}
    boneToModify {boneName=dyn_rope_MSTR}
    rotation=<rot 5 -45 -10>
    translation {z=-0.5}
   }
   n10 = "Transform (Modify) Bone - Bone: dyn_rope_01"(ComponentPose=@n9, Alpha=1, Translation="0,0,0", Rotation="0,0,0", Scale="0.9,1,0.5") {
    rotationMode=BMM_Additive
    rotationSpace=BCS_BoneSpace
    scaleMode=BMM_Replace
    scaleSpace=BCS_BoneSpace
    translationMode=BMM_Additive
    translationSpace=BCS_BoneSpace
    alphaBoolBlend {blendOption=Linear}
    alphaScaleBiasClamp {bInterpResult=true, interpSpeedDecreasing=5, interpSpeedIncreasing=1}
    boneToModify {boneName=dyn_rope_01}
    scale {x=0.9, z=0.5}
   }
   n11 = "Transform (Modify) Bone - Bone: dyn_rope_02"(ComponentPose=@n10, Alpha=1, Translation="0,0,0", Rotation="0,0,0", Scale="0.9,1,1") {
    rotationMode=BMM_Additive
    rotationSpace=BCS_BoneSpace
    scaleMode=BMM_Replace
    scaleSpace=BCS_BoneSpace
    translationMode=BMM_Additive
    translationSpace=BCS_BoneSpace
    alphaBoolBlend {blendOption=Linear}
    alphaScaleBiasClamp {bInterpResult=true, interpSpeedDecreasing=5, interpSpeedIncreasing=1}
    boneToModify {boneName=dyn_rope_02}
    scale {x=0.9}
   }
   n12 = "Transform (Modify) Bone - Bone: dyn_rope_03"(ComponentPose=@n11, Alpha=1, Translation="0,0,0", Rotation="0,0,0", Scale="0.9,1,1") {
    rotationMode=BMM_Additive
    rotationSpace=BCS_BoneSpace
    scaleMode=BMM_Replace
    scaleSpace=BCS_BoneSpace
    translationMode=BMM_Additive
    translationSpace=BCS_BoneSpace
    alphaBoolBlend {blendOption=Linear}
    alphaScaleBiasClamp {bInterpResult=true, interpSpeedDecreasing=5, interpSpeedIncreasing=1}
    boneToModify {boneName=dyn_rope_03}
    scale {x=0.9}
   }
   n13 = "Transform (Modify) Bone - Bone: dyn_rope_01"(ComponentPose=@n12, Alpha=1, Translation="0,0,0", Rotation="0,0,0", Scale="0.8,1,1") {
    rotationMode=BMM_Additive
    rotationSpace=BCS_BoneSpace
    scaleMode=BMM_Replace
    scaleSpace=BCS_BoneSpace
    translationMode=BMM_Additive
    translationSpace=BCS_BoneSpace
    alphaBoolBlend {blendOption=Linear}
    alphaScaleBiasClamp {bInterpResult=true, interpSpeedDecreasing=5, interpSpeedIncreasing=1}
    boneToModify {boneName=dyn_rope_01}
    scale {x=0.8}
   }
   n14 = "Transform (Modify) Bone - Bone: dyn_rope_01"(ComponentPose=@n13, Alpha=1, Translation="0,0,0", Rotation="0,0,0", Scale="0.8,1,1") {
    rotationMode=BMM_Additive
    rotationSpace=BCS_BoneSpace
    scaleMode=BMM_Replace
    scaleSpace=BCS_BoneSpace
    translationMode=BMM_Additive
    translationSpace=BCS_BoneSpace
    alphaBoolBlend {blendOption=Linear}
    alphaScaleBiasClamp {bInterpResult=true, interpSpeedDecreasing=5, interpSpeedIncreasing=1}
    boneToModify {boneName=dyn_rope_01}
    scale {x=0.8}
   }
   Output Pose(Result=Component To Local(ComponentPose=@n14)) {layerGroup=None}
   n15 = Control Rig!(
    Source=Copy Pose From Mesh() {bCopyCurves=true, bUseAttachedParent=true}
    Alpha
    Is Aiming Pistol:double=Get Is Aiming Pistol() -> (Is Aiming Pistol:double)
    Is Aiming Rifle:double=Get Is Aiming Rifle() -> (Is Aiming Rifle:double)
   ) -> (Pose) {
    alphaBoolBlend {blendOption=Linear}
    controlRigAssetReference {blueprintRigClass=&/F_MED_Monarch_CtrlRig.F_MED_Monarch_CtrlRig_C}
   }
   Save cached pose 'PreSim'(Pose=@n15) {cacheName=PreSim}
   n16 = RigidBody!(
    ComponentPose=Local To Component(LocalPose=Use cached pose 'PreSim'())
    Alpha
    OverrideWorldGravity:Vector=Get GravityOverride()
    ExternalForce
   ) -> (Pose) {
    bClampLinearTranslationLimitToRefPose=true
    bOverrideWorldGravity=true
    overridePhysicsAsset=&/F_MED_Monarch_Physics
    simulationSpace=BaseBoneSpace
    alphaBoolBlend {blendOption=Linear}
    baseBoneRef {boneName=pelvis}
    componentAppliedLinearAccClamp=<1000 1000 1000>
    componentLinearAccScale=<0.4 0.45 0.4>
    componentLinearVelScale=<0.4 0.45 0.4>
    overrideWorldGravity {z=-980}
   }
   n17 = Get RopeConstrain() -> (RopeConstrain:double)
   n18 = "Constraint - dyn_rope_02 (PPPPPP)"(
    ComponentPose=@n16
    Alpha=@n17
    ConstraintWeights_0=0.0145
    ConstraintWeights_1=0.022
    ConstraintWeights_2=0.0058
    ConstraintWeights_3=0.1994
    ConstraintWeights_4=0.1213
    ConstraintWeights_5=0.6369
   ) {
    alphaBoolBlend {blendOption=Linear}
    alphaScaleBiasClamp {bInterpResult=true, interpSpeedIncreasing=20}
    boneToModify {boneName=dyn_rope_02}
    constraintSetup [
     {targetBone {boneName=spine_01}}
     {targetBone {boneName=spine_02}}
     {targetBone {boneName=spine_04}}
     {targetBone {boneName=pelvis}}
     {targetBone {boneName=thigh_r}}
     {targetBone {boneName=thigh_twist_01_r}}
    ]
    constraintWeights [1,1,1,1,1,1]
   }
   n19 = "Constraint - dyn_rope_03 (PPPPPPP)"(
    ComponentPose=@n18
    Alpha=@n17
    ConstraintWeights_0=0
    ConstraintWeights_1=0.0007
    ConstraintWeights_2=0.001
    ConstraintWeights_3=0.0002
    ConstraintWeights_4=0.0425
    ConstraintWeights_5=0.0001
    ConstraintWeights_6=0.9556
   ) {
    alphaBoolBlend {blendOption=Linear}
    alphaScaleBiasClamp {bInterpResult=true, interpSpeedIncreasing=20}
    boneToModify {boneName=dyn_rope_03}
    constraintSetup [
     {targetBone {boneName=calf_r}}
     {targetBone {boneName=spine_01}}
     {targetBone {boneName=spine_02}}
     {targetBone {boneName=spine_04}}
     {targetBone {boneName=pelvis}}
     {targetBone {boneName=thigh_r}}
     {targetBone {boneName=thigh_twist_01_r}}
    ]
    constraintWeights [1,1,1,1,1,1,1]
   }
   n20 = "Constraint - dyn_rope_04 (P)"(ComponentPose=@n19, Alpha=@n17, ConstraintWeights_0=1) {
    alphaBoolBlend {blendOption=Linear}
    alphaScaleBiasClamp {bInterpResult=true, interpSpeedIncreasing=20}
    boneToModify {boneName=dyn_rope_04}
    constraintSetup [{targetBone {boneName=thigh_r}}]
    constraintWeights [1]
   }
   n21 = "Transform (Modify) Bone - Bone: dyn_rope_03"(ComponentPose=@n20, Alpha=@n17, Translation="0,0,-3", Rotation="-7,0,0", Scale="1,1,1") {
    rotationMode=BMM_Additive
    rotationSpace=BCS_BoneSpace
    translationMode=BMM_Additive
    translationSpace=BCS_BoneSpace
    alphaBoolBlend {blendOption=Linear}
    alphaScaleBiasClamp {bInterpResult=true, interpSpeedIncreasing=30}
    boneToModify {boneName=dyn_rope_03}
    rotation {pitch=-7}
    translation {z=-3}
   }
   n22 = Component To Local(ComponentPose=@n21)
   n23 = RigidBody(ComponentPose=Local To Component(LocalPose=Use cached pose 'PreSim'()), OverrideWorldGravity=Get GravityOverride()) {
    bClampLinearTranslationLimitToRefPose=true
    bOverrideWorldGravity=true
    overridePhysicsAsset=&/F_MED_Monarch_Crouch_Physics
    simulationSpace=BaseBoneSpace
    alphaBoolBlend {blendOption=Linear}
    baseBoneRef {boneName=pelvis}
    componentAppliedLinearAccClamp=<4000 4000 1500>
    componentLinearAccScale=<0.45 0.45 0.45>
    componentLinearVelScale=<0.45 0.45 0.45>
    overrideWorldGravity {z=-980}
    simSpaceSettings {
     maxAngularAcceleration=1000
     maxAngularVelocity=3
     maxLinearAcceleration=700
     maxLinearVelocity=500
     velocityScaleZ=0
     worldAlpha=0.25
     externalLinearDragV=<0.2 0.2>
    }
   }
   n24 = Component To Local(ComponentPose=@n23)
   n25 = Get CrouchWalking() -> (CrouchWalking:bool)
   n26 = "Transform (Modify) Bone - Bone: dyn_rope_MSTR"!(
    ComponentPose=Local To Component(LocalPose=Blend Poses by bool(BlendPose_0=@n24, BlendPose_1=@n22, BlendTime_0=0.25, BlendTime_1=0.25, bActiveValue=@n25) {blendType=Linear})
    bAlphaBoolEnabled:bool=Get JumpingOrFalling() -> (JumpingOrFalling:bool)
    Translation="0,0,0"
    Rotation="0,0,15"
    Scale="1,1,1"
   ) -> (Pose) {
    alphaInputType=Bool
    rotationMode=BMM_Additive
    rotationSpace=BCS_BoneSpace
    scaleMode=BMM_Additive
    scaleSpace=BCS_BoneSpace
    translationMode=BMM_Additive
    translationSpace=BCS_BoneSpace
    alphaBoolBlend {blendInTime=0.1, blendOption=Linear, blendOutTime=0.1}
    boneToModify {boneName=dyn_rope_MSTR}
    rotation {roll=15}
   }
   n27 = "Transform (Modify) Bone - Bone: dyn_rope_MSTR"(ComponentPose=@n26, bAlphaBoolEnabled=Get NewIsSkydiving() -> (NewIsSkydiving:bool), Translation="0,0,0", Rotation="0,0,-20", Scale="1,1,1") {
    alphaInputType=Bool
    rotationMode=BMM_Additive
    rotationSpace=BCS_BoneSpace
    scaleMode=BMM_Additive
    scaleSpace=BCS_BoneSpace
    translationMode=BMM_Additive
    translationSpace=BCS_BoneSpace
    alphaBoolBlend {blendInTime=0.1, blendOption=Linear, blendOutTime=0.1}
    boneToModify {boneName=dyn_rope_MSTR}
    rotation {roll=-20}
   }
   n28 = RigidBody!(
    ComponentPose=@n27
    Alpha
    OverrideWorldGravity=Get GravityOverride()
    ExternalForce
    SimSpaceSettings:Sim Space Settings Structure=(WorldAlpha=1,VelocityScaleZ=0,DampingAlpha=1,MaxLinearVelocity=500,MaxAngularVelocity=3,MaxLinearAcceleration=700,MaxAngularAcceleration=1000,ExternalLinearDragV=(X=0.1,Y=0.1,Z=0),ExternalLinearVelocity=(X=0,Y=0,Z=0),ExternalAngularVelocity=(X=0,Y=0,Z=0))
   ) -> (Pose) {
    bClampLinearTranslationLimitToRefPose=true
    bOverrideWorldGravity=true
    overridePhysicsAsset=&/F_MED_Monarch_BackFlap_Physics
    simulationSpace=BaseBoneSpace
    alphaBoolBlend {blendOption=Linear}
    baseBoneRef {boneName=pelvis}
    componentAppliedLinearAccClamp=<5000 5000 5000>
    componentLinearAccScale=<0.6 0.4 0.6>
    overrideWorldGravity {z=-980}
    simSpaceSettings {
     maxAngularAcceleration=1000
     maxAngularVelocity=3
     maxLinearAcceleration=700
     maxLinearVelocity=500
     velocityScaleZ=0
     worldAlpha=1
     externalLinearDragV=<0.1 0.1>
    }
   }
   n29 = "Transform (Modify) Bone - Bone: dyn_backFlap_MSTR"!(ComponentPose=@n28, bAlphaBoolEnabled:bool=True, Translation="0,0,1", Rotation="5,0,0", Scale="1,1,1") -> (Pose) {
    alphaInputType=Bool
    rotationMode=BMM_Additive
    rotationSpace=BCS_BoneSpace
    scaleMode=BMM_Additive
    scaleSpace=BCS_BoneSpace
    translationMode=BMM_Additive
    translationSpace=BCS_BoneSpace
    alphaBoolBlend {blendInTime=0.1, blendOption=Linear, blendOutTime=0.1}
    boneToModify {boneName=dyn_backFlap_MSTR}
    rotation {pitch=5}
    translation {z=1}
   }
   n30 = Get isWearingCape() -> (isWearingCape:double)
   n31 = "Constraint - dyn_backFlap (P)"(ComponentPose=@n29, Alpha=@n30, ConstraintWeights_0=1) {
    alphaBoolBlend {blendOption=Linear}
    boneToModify {boneName=dyn_backFlap}
    constraintSetup [{targetBone {boneName=pelvis}}]
    constraintWeights [1]
   }
   n32 = "Constraint - dyn_backFlap_01 (P)"(ComponentPose=@n31, Alpha=@n30, ConstraintWeights_0=1) {
    alphaBoolBlend {blendOption=Linear}
    boneToModify {boneName=dyn_backFlap_01}
    constraintSetup [{targetBone {boneName=pelvis}}]
    constraintWeights [1]
   }
   n33 = "Constraint - dyn_backFlap_02 (P)"(ComponentPose=@n32, Alpha=@n30, ConstraintWeights_0=1) {
    alphaBoolBlend {blendOption=Linear}
    boneToModify {boneName=dyn_backFlap_02}
    constraintSetup [{targetBone {boneName=pelvis}}]
    constraintWeights [1]
   }
   n34 = "Constraint - dyn_backFlap_03 (P)"(ComponentPose=@n33, Alpha=@n30, ConstraintWeights_0=1) {
    alphaBoolBlend {blendOption=Linear}
    boneToModify {boneName=dyn_backFlap_03}
    constraintSetup [{targetBone {boneName=pelvis}}]
    constraintWeights [1]
   }
   n35 = "Constraint - dyn_backFlap_04 (P)"(ComponentPose=@n34, Alpha=@n30, ConstraintWeights_0=1) {
    alphaBoolBlend {blendOption=Linear}
    boneToModify {boneName=dyn_backFlap_04}
    constraintSetup [{targetBone {boneName=pelvis}}]
    constraintWeights [1]
   }
   n36 = Get isWearingCape()
   n37 = "Transform (Modify) Bone - Bone: dyn_backFlap_01"(ComponentPose=@n35, Alpha=@n36, Translation="0,0,0", Rotation="0.000,0.000,-10", Scale="1,1,1") {
    rotationMode=BMM_Additive
    scaleMode=BMM_Additive
    translationMode=BMM_Additive
    alphaBoolBlend {blendOption=Linear}
    boneToModify {boneName=dyn_backFlap_01}
    rotation {roll=-10}
   }
   n38 = "Transform (Modify) Bone - Bone: dyn_backFlap_03"(ComponentPose=@n37, Alpha=@n36, Translation="0,0,0", Rotation="0,0,-18", Scale="1,1,1") {
    rotationMode=BMM_Additive
    scaleMode=BMM_Additive
    translationMode=BMM_Additive
    alphaBoolBlend {blendOption=Linear}
    boneToModify {boneName=dyn_backFlap_03}
    rotation {roll=-18}
   }
   n39 = "Transform (Modify) Bone - Bone: dyn_backFlap_04"(ComponentPose=@n38, Alpha=@n36, Translation="0,0,0", Rotation="0,0,-18", Scale="1,1,1") {
    rotationMode=BMM_Additive
    scaleMode=BMM_Additive
    translationMode=BMM_Additive
    alphaBoolBlend {blendOption=Linear}
    boneToModify {boneName=dyn_backFlap_04}
    rotation {roll=-18}
   }
   n40 = Anim Dynamics - Bone: dyn_backFlap_02!(ComponentPose=@n39, Alpha=1, GravityScale:float=1, GravityOverride:Vector="0,0,0", ExternalForce="0,0,0") -> (Pose) {
    angularDampingOverride=1
    angularSpringConstant=75
    bAngularSpring=true
    bGravityOverrideInSimSpace=true
    bOverrideAngularDamping=true
    simulationSpace=BoneRelative
    alphaBoolBlend {blendOption=Linear}
    boundBone {boneName=dyn_backFlap_02}
    componentAppliedLinearAccClamp=<5000 5000 5000>
    physicsBodyDefinitions [{
      sphereCollisionRadius=0
      boundBone {boneName=dyn_backFlap_02}
      boxExtents=<10 20 20>
      constraintSetup {angularLimitsMax=<5 15 5>, angularLimitsMin=<-5 -5 -5>, angularTarget=<0 0 0>}
      localJointOffset=<2 0 -0.5>
     }]
    relativeSpaceBone {boneName=dyn_root_simspace}
   }
   n41 = Anim Dynamics - Bone: dyn_backFlap_03!(ComponentPose=@n40, Alpha=1, GravityScale:float=1, GravityOverride:Vector="0,0,0", ExternalForce="0,0,0") -> (Pose) {
    angularDampingOverride=1
    angularSpringConstant=75
    bAngularSpring=true
    bGravityOverrideInSimSpace=true
    bOverrideAngularDamping=true
    simulationSpace=BoneRelative
    alphaBoolBlend {blendOption=Linear}
    boundBone {boneName=dyn_backFlap_03}
    componentAppliedLinearAccClamp=<5000 5000 5000>
    physicsBodyDefinitions [{
      sphereCollisionRadius=0
      boundBone {boneName=dyn_backFlap_03}
      boxExtents=<10 20 20>
      constraintSetup {angularLimitsMax=<5 15 5>, angularLimitsMin=<-5 -5 -5>, angularTarget=<0 0 0>}
      localJointOffset=<2 0 -0.5>
     }]
    relativeSpaceBone {boneName=dyn_root_simspace}
   }
   n42 = Anim Dynamics - Bone: dyn_backFlap_04!(ComponentPose=@n41, Alpha=1, GravityScale:float=1, GravityOverride:Vector="0,0,0", ExternalForce="0,0,0") -> (Pose) {
    angularDampingOverride=1
    angularSpringConstant=75
    bAngularSpring=true
    bGravityOverrideInSimSpace=true
    bOverrideAngularDamping=true
    simulationSpace=BoneRelative
    alphaBoolBlend {blendOption=Linear}
    boundBone {boneName=dyn_backFlap_04}
    componentAppliedLinearAccClamp=<5000 5000 5000>
    physicsBodyDefinitions [{
      sphereCollisionRadius=0
      boundBone {boneName=dyn_backFlap_04}
      boxExtents=<10 20 20>
      constraintSetup {angularLimitsMax=<5 15 5>, angularLimitsMin=<-5 -5 -5>, angularTarget=<0 0 0>}
      localJointOffset=<2 0 -0.5>
     }]
    relativeSpaceBone {boneName=dyn_root_simspace}
   }
   n43 = "Constraint - dyn_frontFlap_MSTR (P)"!(ComponentPose=@n42, bAlphaBoolEnabled:bool=True, ConstraintWeights_0=1) -> (Pose) {
    alphaInputType=Bool
    alphaBoolBlend {blendOption=Linear}
    boneToModify {boneName=dyn_frontFlap_MSTR}
    constraintSetup [{targetBone {boneName=pelvis}}]
    constraintWeights [1]
   }
   n44 = "Constraint - dyn_frontFlap_01 (P)"!(ComponentPose=@n43, bAlphaBoolEnabled:bool=True, ConstraintWeights_0=1) -> (Pose) {
    alphaInputType=Bool
    alphaBoolBlend {blendOption=Linear}
    boneToModify {boneName=dyn_frontFlap_01}
    constraintSetup [{targetBone {boneName=pelvis}}]
    constraintWeights [1]
   }
   n45 = "Constraint - dyn_frontFlap_02 (P)"!(ComponentPose=@n44, bAlphaBoolEnabled:bool=True, ConstraintWeights_0=1) -> (Pose) {
    alphaInputType=Bool
    alphaBoolBlend {blendOption=Linear}
    boneToModify {boneName=dyn_frontFlap_02}
    constraintSetup [{targetBone {boneName=pelvis}}]
    constraintWeights [1]
   }
   n46 = "Constraint - dyn_frontFlap_03 (P)"!(ComponentPose=@n45, bAlphaBoolEnabled:bool=True, ConstraintWeights_0=1) -> (Pose) {
    alphaInputType=Bool
    alphaBoolBlend {blendOption=Linear}
    boneToModify {boneName=dyn_frontFlap_03}
    constraintSetup [{targetBone {boneName=pelvis}}]
    constraintWeights [1]
   }
   n47 = "Constraint - dyn_frontFlap_04 (P)"!(ComponentPose=@n46, bAlphaBoolEnabled:bool=True, ConstraintWeights_0=1) -> (Pose) {
    alphaInputType=Bool
    alphaBoolBlend {blendOption=Linear}
    boneToModify {boneName=dyn_frontFlap_04}
    constraintSetup [{targetBone {boneName=pelvis}}]
    constraintWeights [1]
   }
   n48 = Anim Dynamics - Bone: dyn_frontFlap_01!(ComponentPose=@n47, bAlphaBoolEnabled:bool=True, ExternalForce="0,0,0") -> (Pose) {
    alphaInputType=Bool
    angularDampingOverride=0.9999
    angularSpringConstant=75
    bAngularSpring=true
    bGravityOverrideInSimSpace=true
    bOverrideAngularDamping=true
    simulationSpace=RootRelative
    alphaBoolBlend {blendOption=Linear}
    boundBone {boneName=dyn_frontFlap_01}
    componentLinearAccScale=<1 1 1>
    componentLinearVelScale=<1 1 1>
    physicsBodyDefinitions [{
      sphereCollisionRadius=0
      boundBone {boneName=dyn_frontFlap_01}
      boxExtents=<20 20 10>
      constraintSetup {angularLimitsMax=<25 0 25>, angularLimitsMin=<-25 -25 -25>, angularTarget=<0 0 0>}
      localJointOffset=<2 0 0>
     }]
   }
   n49 = Anim Dynamics - Bone: dyn_frontFlap_02!(ComponentPose=@n48, bAlphaBoolEnabled:bool=True, ExternalForce="0,0,0") -> (Pose) {
    alphaInputType=Bool
    angularDampingOverride=0.9999
    angularSpringConstant=75
    bAngularSpring=true
    bGravityOverrideInSimSpace=true
    bOverrideAngularDamping=true
    simulationSpace=RootRelative
    alphaBoolBlend {blendOption=Linear}
    boundBone {boneName=dyn_frontFlap_02}
    componentLinearAccScale=<1 1 1>
    componentLinearVelScale=<1 1 1>
    physicsBodyDefinitions [{
      sphereCollisionRadius=0
      boundBone {boneName=dyn_frontFlap_02}
      boxExtents=<20 20 10>
      constraintSetup {angularLimitsMax=<25 0 25>, angularLimitsMin=<-25 -25 -25>, angularTarget=<0 0 0>}
      localJointOffset=<2 0 0>
     }]
   }
   n50 = Anim Dynamics - Bone: dyn_frontFlap_03!(ComponentPose=@n49, bAlphaBoolEnabled:bool=True, ExternalForce="0,0,0") -> (Pose) {
    alphaInputType=Bool
    angularDampingOverride=0.9999
    angularSpringConstant=75
    bAngularSpring=true
    bGravityOverrideInSimSpace=true
    bOverrideAngularDamping=true
    simulationSpace=RootRelative
    alphaBoolBlend {blendOption=Linear}
    boundBone {boneName=dyn_frontFlap_03}
    componentLinearAccScale=<1 1 1>
    componentLinearVelScale=<1 1 1>
    physicsBodyDefinitions [{
      sphereCollisionRadius=0
      boundBone {boneName=dyn_frontFlap_03}
      boxExtents=<20 20 10>
      constraintSetup {angularLimitsMax=<25 0 25>, angularLimitsMin=<-25 -25 -25>, angularTarget=<0 0 0>}
      localJointOffset=<2 0 0>
     }]
   }
   n51 = Anim Dynamics - Bone: dyn_frontFlap_04!(ComponentPose=@n50, bAlphaBoolEnabled:bool=True, ExternalForce="0,0,0") -> (Pose) {
    alphaInputType=Bool
    angularDampingOverride=0.9999
    angularSpringConstant=75
    bAngularSpring=true
    bGravityOverrideInSimSpace=true
    bOverrideAngularDamping=true
    simulationSpace=RootRelative
    alphaBoolBlend {blendOption=Linear}
    boundBone {boneName=dyn_frontFlap_04}
    componentLinearAccScale=<1 1 1>
    componentLinearVelScale=<1 1 1>
    physicsBodyDefinitions [{
      sphereCollisionRadius=0
      boundBone {boneName=dyn_frontFlap_04}
      boxExtents=<20 20 10>
      constraintSetup {angularLimitsMax=<25 0 25>, angularLimitsMin=<-25 -25 -25>, angularTarget=<0 0 0>}
      localJointOffset=<2 0 0>
     }]
   }
   n52 = Get Zipline() -> (Zipline:bool)
   n53 = "Transform (Modify) Bone - Bone: dyn_backFlap_MSTR"(ComponentPose=@n51, bAlphaBoolEnabled=@n52, Translation="0,0,0", Rotation="20,0,0", Scale="1,1,1") {
    alphaInputType=Bool
    rotationMode=BMM_Additive
    rotationSpace=BCS_BoneSpace
    scaleMode=BMM_Additive
    scaleSpace=BCS_BoneSpace
    translationMode=BMM_Additive
    translationSpace=BCS_BoneSpace
    alphaBoolBlend {blendInTime=0.1, blendOption=Linear, blendOutTime=0.1}
    boneToModify {boneName=dyn_backFlap_MSTR}
    rotation {pitch=20}
   }
   n54 = "Constraint - dyn_frontFlap_01 (P)"(ComponentPose=@n53, bAlphaBoolEnabled=@n52, ConstraintWeights_0=1) {
    alphaInputType=Bool
    alphaBoolBlend {blendOption=Linear}
    boneToModify {boneName=dyn_frontFlap_01}
    constraintSetup [{targetBone {boneName=pelvis}}]
    constraintWeights [1]
   }
   n55 = "Constraint - dyn_frontFlap_02 (P)"(ComponentPose=@n54, bAlphaBoolEnabled=@n52, ConstraintWeights_0=1) {
    alphaInputType=Bool
    alphaBoolBlend {blendOption=Linear}
    boneToModify {boneName=dyn_frontFlap_02}
    constraintSetup [{targetBone {boneName=pelvis}}]
    constraintWeights [1]
   }
   n56 = "Constraint - dyn_frontFlap_03 (P)"(ComponentPose=@n55, bAlphaBoolEnabled=@n52, ConstraintWeights_0=1) {
    alphaInputType=Bool
    alphaBoolBlend {blendOption=Linear}
    boneToModify {boneName=dyn_frontFlap_03}
    constraintSetup [{targetBone {boneName=pelvis}}]
    constraintWeights [1]
   }
   n57 = "Constraint - dyn_frontFlap_04 (P)"(ComponentPose=@n56, bAlphaBoolEnabled=@n52, ConstraintWeights_0=1) {
    alphaInputType=Bool
    alphaBoolBlend {blendOption=Linear}
    boneToModify {boneName=dyn_frontFlap_04}
    constraintSetup [{targetBone {boneName=pelvis}}]
    constraintWeights [1]
   }
   n58 = "Constraint - dyn_rope_02 (PPPPPP)"(
    ComponentPose=@n57
    Alpha=1
    ConstraintWeights_0=0.0145
    ConstraintWeights_1=0.022
    ConstraintWeights_2=0.0058
    ConstraintWeights_3=0.1994
    ConstraintWeights_4=0.1213
    ConstraintWeights_5=0.6369
   ) {
    alphaBoolBlend {blendOption=Linear}
    alphaScaleBiasClamp {bInterpResult=true, interpSpeedIncreasing=20}
    boneToModify {boneName=dyn_rope_02}
    constraintSetup [
     {targetBone {boneName=spine_01}}
     {targetBone {boneName=spine_02}}
     {targetBone {boneName=spine_04}}
     {targetBone {boneName=pelvis}}
     {targetBone {boneName=thigh_r}}
     {targetBone {boneName=thigh_twist_01_r}}
    ]
    constraintWeights [1,1,1,1,1,1]
   }
   n59 = "Constraint - dyn_rope_03 (PPPPPPP)"(
    ComponentPose=@n58
    Alpha=1
    ConstraintWeights_0=0
    ConstraintWeights_1=0.0007
    ConstraintWeights_2=0.001
    ConstraintWeights_3=0.0002
    ConstraintWeights_4=0.0425
    ConstraintWeights_5=0.0001
    ConstraintWeights_6=0.9556
   ) {
    alphaBoolBlend {blendOption=Linear}
    alphaScaleBiasClamp {bInterpResult=true, interpSpeedIncreasing=20}
    boneToModify {boneName=dyn_rope_03}
    constraintSetup [
     {targetBone {boneName=calf_r}}
     {targetBone {boneName=spine_01}}
     {targetBone {boneName=spine_02}}
     {targetBone {boneName=spine_04}}
     {targetBone {boneName=pelvis}}
     {targetBone {boneName=thigh_r}}
     {targetBone {boneName=thigh_twist_01_r}}
    ]
    constraintWeights [1,1,1,1,1,1,1]
   }
   n60 = "Transform (Modify) Bone - Bone: dyn_rope_02"(ComponentPose=@n59, Alpha=1, Translation="0,0,0", Rotation="-10,0,0", Scale="1,1,1") {
    rotationMode=BMM_Additive
    rotationSpace=BCS_BoneSpace
    scaleSpace=BCS_BoneSpace
    translationMode=BMM_Additive
    translationSpace=BCS_BoneSpace
    alphaBoolBlend {blendOption=Linear}
    alphaScaleBiasClamp {
     bClampResult=true
     bInterpResult=true
     bMapRange=true
     interpSpeedDecreasing=3
     interpSpeedIncreasing=3
     inRange {max=0.015, min=0.03}
     outRange {max=0, min=0.5}
    }
    boneToModify {boneName=dyn_rope_02}
    rotation {pitch=-10}
   }
   n61 = "Transform (Modify) Bone - Bone: dyn_rope_02"!(ComponentPose=@n60, bAlphaBoolEnabled:bool=Get bIsKneeSliding(), Translation="0,0,2", Rotation="10,0,0", Scale="1,1,1") -> (Pose) {
    alphaInputType=Bool
    rotationMode=BMM_Additive
    rotationSpace=BCS_BoneSpace
    translationMode=BMM_Additive
    translationSpace=BCS_BoneSpace
    alphaBoolBlend {blendOption=Linear}
    alphaScaleBiasClamp {interpSpeedIncreasing=30}
    boneToModify {boneName=dyn_rope_02}
    rotation {pitch=10}
    translation {z=2}
   }
   Save cached pose 'PostSim'(Pose=Component To Local(ComponentPose=@n61)) {cacheName=PostSim}
  ]}
 EventGraph {nodes [
   n1 = Event BlueprintUpdateAnimation()
   > Set Delta Time X(Delta Time X=@n1.DeltaTimeX)
   > n2 = Sequence!(execute) -> (then_0, then_1, then_2:Exec, then_3:Exec, then_4:Exec, then_5:Exec)
   Set JumpingOrFalling(execute=@n2.then_2, JumpingOrFalling:bool=OR Boolean(A=Get bIsJumping(), B=Get bIsFalling())) -> (Output_Get:bool)
   > Set CrouchWalking(CrouchWalking:bool=AND Boolean(A=Get bIsAccelerating2D(), B=Get bIsCrouching())) -> (Output_Get:bool)
   > Set NewIsSkydiving(NewIsSkydiving:bool=Get bIsSkyDiving()) -> (Output_Get:bool)
   n3 = Gravity Override State Function DEPRECATED(
    execute=@n2.then_3
    CharacterPartAnimInstance=Self-Reference()
    GameTimeInSeconds=GetGameTimeInSeconds()
    GravityOverrideParams=Get Gravity Override Params()
   )
   > Set GravityOverride(GravityOverride=@n3)
   n4 = Sequence!(execute=@n2.then_4) -> (then_0, then_1, then_2)
   > n5 = Branch(Condition=Get bIsParachuteOpened())
   > Set Sim Space Settings(Sim Space Settings:Sim Space Settings Structure=Make Sim Space Settings!(
     WorldAlpha:float=0
     VelocityScaleZ:float=0
     DampingAlpha:float=1
     MaxLinearVelocity:float=500
     MaxAngularVelocity:float=0.3
     MaxLinearAcceleration:float=700
     MaxAngularAcceleration:float=1000
     ExternalLinearDragV:Vector="0.5,0.5,0"
     ExternalLinearVelocity:Vector="0,0,0"
     ExternalAngularVelocity:Vector="0,0,0"
    ) -> (SimSpaceSettings:Sim Space Settings Structure) {is=title K2Node_MakeStruct}) -> (Output_Get:Sim Space Settings Structure)
   Set Sim Space Settings(
    execute=@n5.else
    Sim Space Settings=Make Sim Space Settings(
     WorldAlpha=0.5
     VelocityScaleZ=0
     DampingAlpha=1
     MaxLinearVelocity=500
     MaxAngularVelocity=0.3
     MaxLinearAcceleration=700
     MaxAngularAcceleration=1000
     ExternalLinearDragV="0.5,0.5,0"
     ExternalLinearVelocity="0,0,0"
     ExternalAngularVelocity="0,0,0"
    )
   )
   n6 = Get Custom Base Mesh() -> (Custom Base Mesh:Skeletal Mesh Component Object Reference)
   n7 = GetBoneTransform!(self:Skinned Mesh Component Object Reference=@n6, InBoneName:Name=dyn_rope_04, TransformSpace:ERelativeTransformSpace Enum=RTS_Component) -> (ReturnValue_Location:Vector, ReturnValue_Rotation:Rotator, ReturnValue_Scale:Vector)
   n8 = GetBoneTransform(self=@n6, InBoneName=hand_r, TransformSpace=RTS_Component)
   Set RopeConstrain(
    execute=@n4.then_1
    RopeConstrain:double=MapRangeClamped(Value="Distance (Vector)"!(V1:Vector=@n8, V2:Vector=@n7) -> (ReturnValue:double), InRangeA=25, InRangeB=10)
   ) -> (Output_Get:double)
   n9 = "Absolute (Float)"(A=float + float(A=Get PawnMovementDirectionAzimuth() -> (PawnMovementDirectionAzimuth:float), B=45))
   n10 = "float < float"(A=Get PawnMovementDirectionAzimuth(), B=-10)
   n11 = Get bIsMoving2D()
   > Set MovingForwardLeft(MovingForwardLeft:double=SelectFloat(A=MapRangeClamped(Value=@n9, InRangeA=0, InRangeB=45, OutRangeA=1, OutRangeB=0), B=0, bPickA=AND Boolean(A=@n10, B=@n11))) -> (Output_Get:double)
   n12 = "Absolute (Float)"(A=Get PawnMovementDirectionAzimuth())
   > Set MovingForward(MovingForward:double=SelectFloat(A=MapRangeClamped(Value=@n12, InRangeA=0, InRangeB=45, OutRangeA=1, OutRangeB=0), B=0, bPickA=Get bIsMoving2D())) -> (Output_Get:double)
   n13 = Get PawnMovementDirectionAzimuth()
   n14 = MapRangeClamped(Value="Absolute (Float)"(A=float + float(A=@n13, B=90)), InRangeA=0, InRangeB=45, OutRangeA=1, OutRangeB=0)
   > Set MovingLeft(MovingLeft:double=@n14) -> (Output_Get:double)
   n15 = Event BlueprintProcessAnimInputEvent()
   > n16 = Branch(Condition=IsNamed(self=@n15.AnimInputEvent, InEventName=aie.Player.Zipline.Entered))
   > Set Zipline(Zipline:bool=true) -> (Output_Get:bool)
   n17 = Branch(execute=@n16.else, Condition=IsNamed(self=@n15.AnimInputEvent, InEventName=aie.Player.Zipline.Exited))
   > Set Zipline(Zipline=false)
   Set Zipline(execute=@n17.else, Zipline=false)
   n18 = Event OnCharacterCustomizationLoaded()
   > Set Custom Base Mesh(Custom Base Mesh:Skeletal Mesh Component Object Reference=Get Array Item with EFort Custom Part Type(Array=@n18.CustomMeshParts)) -> (Output_Get:Skeletal Mesh Component Object Reference)
   n19 = CheckForBackbling()
   n20 = NOT Boolean(A=@n19.bIsBackblingLong)
   > Set isWearingBackpack(isWearingBackpack:double="To Float (Boolean)"(InBool=AND Boolean(A=@n19, B=@n20))) -> (Output_Get:double)
   > Set isNotWearingBackbling(isNotWearingBackbling:double="To Float (Boolean)"(InBool=NOT Boolean(A=@n19))) -> (Output_Get:double)
   > Set isWearingCape(isWearingCape:double="To Float (Boolean)"(InBool=@n19.bIsBackblingLong)) -> (Output_Get:double)
   > Set WearingCape(WearingCape:bool=true) -> (Output_Get:bool)
   Event OnWeaponChanged()
   > n21 = Sequence!(execute) -> (then_0, then_1, then_2, then_3:Exec, then_4:Exec)
   > n22 = Cast To FortPlayerAnimInstance(Object=GetAnimInstance(self=GetOwnerSkeletalMeshComponent()))
   > Set isPistol(isPistol=false)
   > Set isRifle(isRifle:bool=false) -> (Output_Get:bool)
   > n23 = Switch on EFortWeaponCoreAnimation!(execute, Selection:EFortWeaponCoreAnimation Enum=Get WeaponCoreAnim(self=@n22)) -> (
    Melee:Exec
    Pistol:Exec
    Shotgun:Exec
    PaperBlueprint:Exec
    Rifle:Exec
    MeleeOneHand:Exec
    MachinePistol:Exec
    RocketLauncher:Exec
    GrenadeLauncher:Exec
    GoingCommando:Exec
    AssaultRifle:Exec
    TacticalShotgun:Exec
    SniperRifle:Exec
    TrapPlacement:Exec
    ShoulderLauncher:Exec
    AbilityDecoTool:Exec
    Crossbow:Exec
    C4:Exec
    RemoteControl:Exec
    DualWield:Exec
    AR_BullPup:Exec
    AR_ForwardGrip:Exec
    MedPackPaddles:Exec
    SMG_P90:Exec
    AR_DrumGun:Exec
    Consumable_Small:Exec
    Consumable_Large:Exec
    Balloon:Exec
    MountedTurret:Exec
    CreativeTool:Exec
    ExplosiveBow:Exec
    AshtonIndigo:Exec
    AshtonChicago:Exec
    MeleeDualWield:Exec
    Unarmed:Exec
   )
   Set isRifle(execute=[@n23.AssaultRifle, @n23.Rifle, @n23.RocketLauncher, @n23.Shotgun, @n23.SniperRifle, @n23.TacticalShotgun], isRifle=true)
   Set isPistol(execute=@n23.Pistol, isPistol=true)
   n24 = Get bIsTargeting()
   n25 = Get isRifle() -> (isRifle:bool)
   Set Is Aiming Rifle(execute=@n21.then_1, Is Aiming Rifle:double="To Float (Boolean)"(InBool=AND Boolean(A=@n25, B=@n24))) -> (Output_Get:double)
   n26 = Get isPistol() -> (isPistol:bool)
   n27 = Get bIsTargeting()
   Set Is Aiming Pistol(execute=@n21.then_2, Is Aiming Pistol:double="To Float (Boolean)"(InBool=AND Boolean(A=@n26, B=@n27))) -> (Output_Get:double)
   SelectFloat(A=@n14, B=0, bPickA="float < float"(A=Get PawnMovementDirectionAzimuth(), B=-10))
  ]}
]
variables [
 CrouchWalking {type=boolean}
 Custom Base Mesh {type=SkeletalMeshComponent}
 Delta Time X {type=number}
 Gravity Override Params {type=GravityOverrideParamsStruct, default {skydive_Parachuting {gravityOverride=<0 -800 -2550>}}}
 GravityOverride {type=Vector}
 Is Aiming Pistol {type=number, default=1}
 Is Aiming Rifle {type=number, default=1}
 isNotWearingBackbling {type=number}
 isPistol {type=boolean}
 isRifle {type=boolean}
 isWearingBackpack {type=number}
 isWearingCape {type=number}
 JumpingOrFalling {type=boolean}
 MovingForward {type=number}
 MovingForwardLeft {type=number}
 MovingLeft {type=number}
 NewIsSkydiving {type=boolean}
 RopeConstrain {type=number}
 Sim Space Settings {
  type=SimSpaceSettings
  default {
   dampingAlpha=1
   maxAngularAcceleration=10000
   maxAngularVelocity=10000
   maxLinearAcceleration=10000
   maxLinearVelocity=10000
   velocityScaleZ=1
   worldAlpha=0
   externalLinearDragV=<0 0 0>
  }
 }
 WearingCape {type=boolean}
 Zipline {type=boolean}
]
