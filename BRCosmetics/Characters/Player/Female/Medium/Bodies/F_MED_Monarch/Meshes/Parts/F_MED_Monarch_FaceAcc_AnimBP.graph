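# Post-process style hair AnimBP for the F_MED_Monarch FaceAcc mesh, adding secondary motion to the hair bones on top of the pose copied from the attached parent mesh.
# AnimDynamics bodies simulate hair strands relative to dyn_root_simspace, whose world rotation comes from sim_space_rotation, with constraints softening the result.
# When isWearingCape is true, the cached PostSim pose gets extra constraints toward spine, clavicles and head, plus bone scale tweaks.
# Final constraints and an additive transform adjust the bangs.
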AnimBlueprint /BRCosmetics/Characters/Player/Female/Medium/Bodies/F_MED_Monarch/Meshes/Parts/F_MED_Monarch_FaceAcc_AnimBP
target_skeleton=&/F_MED_Monarch_FaceAcc_Skeleton
graphs [
 AnimGraph {nodes [
   n1 = "Constraint - dyn_hair_l_main_c_02 (P)"(ComponentPose=Local To Component(LocalPose=Use cached pose 'PostSim'()), Alpha=0.5, ConstraintWeights_0=1) {
    alphaBoolBlend {blendOption=Linear}
    boneToModify {boneName=dyn_hair_l_main_c_02}
    constraintSetup [{targetBone {boneName=spine_05}}]
    constraintWeights [1]
   }
   n2 = "Constraint - dyn_hair_l_main_c_03 (P)"(ComponentPose=@n1, Alpha=0.45, ConstraintWeights_0=1) {
    alphaBoolBlend {blendOption=Linear}
    boneToModify {boneName=dyn_hair_l_main_c_03}
    constraintSetup [{targetBone {boneName=clavicle_l}}]
    constraintWeights [1]
   }
   n3 = "Constraint - dyn_hair_l_main_c_04 (P)"(ComponentPose=@n2, Alpha=0.4, ConstraintWeights_0=1) {
    alphaBoolBlend {blendOption=Linear}
    boneToModify {boneName=dyn_hair_l_main_c_04}
    constraintSetup [{targetBone {boneName=spine_05}}]
    constraintWeights [1]
   }
   n4 = "Constraint - dyn_hair_l_main_c_05 (P)"(ComponentPose=@n3, Alpha=0.3, ConstraintWeights_0=1) {
    alphaBoolBlend {blendOption=Linear}
    boneToModify {boneName=dyn_hair_l_main_c_05}
    constraintSetup [{targetBone {boneName=spine_05}}]
    constraintWeights [1]
   }
   n5 = "Constraint - dyn_hair_l_main_d_01 (P)"(ComponentPose=@n4, Alpha=0.4, ConstraintWeights_0=1) {
    alphaBoolBlend {blendOption=Linear}
    boneToModify {boneName=dyn_hair_l_main_d_01}
    constraintSetup [{targetBone {boneName=spine_05}}]
    constraintWeights [1]
   }
   n6 = "Constraint - dyn_hair_l_main_d_02 (P)"(ComponentPose=@n5, Alpha=0.1, ConstraintWeights_0=1) {
    alphaBoolBlend {blendOption=Linear}
    boneToModify {boneName=dyn_hair_l_main_d_02}
    constraintSetup [{targetBone {boneName=spine_05}}]
    constraintWeights [1]
   }
   n7 = "Constraint - dyn_hair_c_main_01 (P)"(ComponentPose=@n6, Alpha=0.6, ConstraintWeights_0=1) {
    alphaBoolBlend {blendOption=Linear}
    boneToModify {boneName=dyn_hair_c_main_01}
    constraintSetup [{targetBone {boneName=spine_05}}]
    constraintWeights [1]
   }
   n8 = "Constraint - dyn_hair_c_main_02 (P)"(ComponentPose=@n7, Alpha=0.4, ConstraintWeights_0=1) {
    alphaBoolBlend {blendOption=Linear}
    boneToModify {boneName=dyn_hair_c_main_02}
    constraintSetup [{targetBone {boneName=spine_05}}]
    constraintWeights [1]
   }
   n9 = "Constraint - dyn_hair_c_main_03 (P)"(ComponentPose=@n8, Alpha=0.4, ConstraintWeights_0=1) {
    alphaBoolBlend {blendOption=Linear}
    boneToModify {boneName=dyn_hair_c_main_03}
    constraintSetup [{targetBone {boneName=spine_05}}]
    constraintWeights [1]
   }
   n10 = "Constraint - dyn_hair_c_main_04 (P)"(ComponentPose=@n9, Alpha=0.5, ConstraintWeights_0=1) {
    alphaBoolBlend {blendOption=Linear}
    boneToModify {boneName=dyn_hair_c_main_04}
    constraintSetup [{targetBone {boneName=spine_05}}]
    constraintWeights [1]
   }
   n11 = "Constraint - dyn_hair_c_main_05 (P)"(ComponentPose=@n10, Alpha=0.6, ConstraintWeights_0=1) {
    alphaBoolBlend {blendOption=Linear}
    boneToModify {boneName=dyn_hair_c_main_05}
    constraintSetup [{targetBone {boneName=spine_05}}]
    constraintWeights [1]
   }
   n12 = "Constraint - dyn_hair_r_main_03 (PPP)"!(ComponentPose=@n11, Alpha=0.5, ConstraintWeights_0=1, ConstraintWeights_1:float=0, ConstraintWeights_2:float=0) -> (Pose) {
    alphaBoolBlend {blendOption=Linear}
    boneToModify {boneName=dyn_hair_r_main_03}
    constraintSetup [{targetBone {boneName=spine_05}}, {targetBone {boneName=dyn_hair_c_main_02}}, {targetBone {boneName=dyn_hair_c_main_03}}]
    constraintWeights [1,1,1]
   }
   n13 = "Constraint - dyn_hair_r_main_04 (PR)"!(ComponentPose=@n12, Alpha=0.45, ConstraintWeights_0=1, ConstraintWeights_1:float=0) -> (Pose) {
    alphaBoolBlend {blendOption=Linear}
    boneToModify {boneName=dyn_hair_r_main_04}
    constraintSetup [{targetBone {boneName=spine_05}}, {transformType=Rotation, targetBone {boneName=clavicle_r}}]
    constraintWeights [1,1]
   }
   n14 = "Constraint - dyn_hair_r_main_05 (RR)"!(ComponentPose=@n13, Alpha=1, ConstraintWeights_0=0.2, ConstraintWeights_1:float=0.8) -> (Pose) {
    alphaBoolBlend {blendOption=Linear}
    boneToModify {boneName=dyn_hair_r_main_05}
    constraintSetup [{transformType=Rotation, targetBone {boneName=spine_05}}, {transformType=Rotation, targetBone {boneName=dyn_hair_c_main_04}}]
    constraintWeights [1,1]
   }
   n15 = "Constraint - dyn_hair_r_main_06 (R)"(ComponentPose=@n14, Alpha=0.5, ConstraintWeights_0=1) {
    alphaBoolBlend {blendOption=Linear}
    boneToModify {boneName=dyn_hair_r_main_06}
    constraintSetup [{transformType=Rotation, targetBone {boneName=dyn_hair_c_main_05}}]
    constraintWeights [1]
   }
   n16 = "Constraint - dyn_hair_r_main_07 (R)"(ComponentPose=@n15, Alpha=0.5, ConstraintWeights_0=1) {
    alphaBoolBlend {blendOption=Linear}
    boneToModify {boneName=dyn_hair_r_main_07}
    constraintSetup [{transformType=Rotation, targetBone {boneName=dyn_hair_c_main_05}}]
    constraintWeights [1]
   }
   n17 = "Constraint - dyn_hair_r_main_02 (R)"(ComponentPose=@n16, Alpha=0.5, ConstraintWeights_0=1) {
    alphaBoolBlend {blendOption=Linear}
    boneToModify {boneName=dyn_hair_r_main_02}
    constraintSetup [{transformType=Rotation, targetBone {boneName=clavicle_r}}]
    constraintWeights [1]
   }
   n18 = "Constraint - dyn_hair_r_main_03 (PP)"!(ComponentPose=@n17, Alpha=0.5, ConstraintWeights_0=1, ConstraintWeights_1:float=0.5) -> (Pose) {
    alphaBoolBlend {blendOption=Linear}
    boneToModify {boneName=dyn_hair_r_main_03}
    constraintSetup [{targetBone {boneName=clavicle_r}}, {targetBone {boneName=spine_05}}]
    constraintWeights [1,1]
   }
   n19 = "Transform (Modify) Bone - Bone: dyn_hair_bang_b_04"(ComponentPose=@n18, Alpha=1, Translation="0,0,0", Rotation="5.575,15.803,0.607", Scale="1,1,1") {
    rotationMode=BMM_Additive
    rotationSpace=BCS_BoneSpace
    scaleMode=BMM_Additive
    scaleSpace=BCS_BoneSpace
    translationMode=BMM_Additive
    translationSpace=BCS_BoneSpace
    alphaBoolBlend {blendOption=Linear}
    boneToModify {boneName=dyn_hair_bang_b_04}
    rotation=<rot 5.575 15.803 0.607>
   }
   n20 = "Constraint - dyn_hair_l_main (P)"(ComponentPose=@n19, Alpha=0.2, ConstraintWeights_0=1) {
    alphaBoolBlend {blendOption=Linear}
    boneToModify {boneName=dyn_hair_l_main}
    constraintSetup [{targetBone {boneName=head}}]
    constraintWeights [1]
   }
   n21 = "Constraint - dyn_hair_l_main_c_01 (P)"(ComponentPose=@n20, Alpha=0.1, ConstraintWeights_0=1) {
    alphaBoolBlend {blendOption=Linear}
    boneToModify {boneName=dyn_hair_l_main_c_01}
    constraintSetup [{targetBone {boneName=head}}]
    constraintWeights [1]
   }
   n22 = "Constraint - dyn_hair_l_main_b_02 (P)"(ComponentPose=@n21, Alpha=0.3, ConstraintWeights_0=1) {
    alphaBoolBlend {blendOption=Linear}
    boneToModify {boneName=dyn_hair_l_main_b_02}
    constraintSetup [{targetBone {boneName=clavicle_l}}]
    constraintWeights [1]
   }
   n23 = "Constraint - dyn_hair_l_main_b_03 (P)"(ComponentPose=@n22, Alpha=0.2, ConstraintWeights_0=1) {
    alphaBoolBlend {blendOption=Linear}
    boneToModify {boneName=dyn_hair_l_main_b_03}
    constraintSetup [{targetBone {boneName=clavicle_l}}]
    constraintWeights [1]
   }
   n24 = "Constraint - dyn_hair_l_main_b_04 (P)"(ComponentPose=@n23, Alpha=0.3, ConstraintWeights_0=1) {
    alphaBoolBlend {blendOption=Linear}
    boneToModify {boneName=dyn_hair_l_main_b_04}
    constraintSetup [{targetBone {boneName=clavicle_l}}]
    constraintWeights [1]
   }
   n25 = "Constraint - dyn_hair_l_main_b_05 (P)"(ComponentPose=@n24, Alpha=0.4, ConstraintWeights_0=1) {
    alphaBoolBlend {blendOption=Linear}
    boneToModify {boneName=dyn_hair_l_main_b_05}
    constraintSetup [{targetBone {boneName=clavicle_l}}]
    constraintWeights [1]
   }
   n26 = "Transform (Modify) Bone - Bone: dyn_hair_r_main_03"(ComponentPose=@n25, Alpha=1, Translation="0,0,0", Rotation="0,0,0", Scale="0.8,0.8,0.8") {
    rotationMode=BMM_Additive
    rotationSpace=BCS_BoneSpace
    scaleMode=BMM_Additive
    scaleSpace=BCS_BoneSpace
    translationMode=BMM_Additive
    translationSpace=BCS_BoneSpace
    alphaBoolBlend {blendOption=Linear}
    alphaScaleBiasClamp {bInterpResult=true, interpSpeedDecreasing=20, interpSpeedIncreasing=20}
    boneToModify {boneName=dyn_hair_r_main_03}
    scale=<0.8 0.8 0.8>
   }
   n27 = "Transform (Modify) Bone - Bone: dyn_hair_r_main_04"(ComponentPose=@n26, Alpha=1, Translation="0,0,0", Rotation="0,0,0", Scale="0.7,0.7,0.7") {
    rotationMode=BMM_Additive
    rotationSpace=BCS_BoneSpace
    scaleMode=BMM_Additive
    scaleSpace=BCS_BoneSpace
    translationMode=BMM_Additive
    translationSpace=BCS_BoneSpace
    alphaBoolBlend {blendOption=Linear}
    alphaScaleBiasClamp {bInterpResult=true, interpSpeedDecreasing=20, interpSpeedIncreasing=20}
    boneToModify {boneName=dyn_hair_r_main_04}
    scale=<0.7 0.7 0.7>
   }
   n28 = "Transform (Modify) Bone - Bone: dyn_hair_r_main_05"(ComponentPose=@n27, Alpha=1, Translation="0,0,0", Rotation="0,0,0", Scale="0.6,0.6,0.6") {
    rotationMode=BMM_Additive
    rotationSpace=BCS_BoneSpace
    scaleMode=BMM_Additive
    scaleSpace=BCS_BoneSpace
    translationMode=BMM_Additive
    translationSpace=BCS_BoneSpace
    alphaBoolBlend {blendOption=Linear}
    alphaScaleBiasClamp {bInterpResult=true, interpSpeedDecreasing=20, interpSpeedIncreasing=20}
    boneToModify {boneName=dyn_hair_r_main_05}
    scale=<0.6 0.6 0.6>
   }
   n29 = "Transform (Modify) Bone - Bone: dyn_hair_c_main_02"(ComponentPose=@n28, Alpha=1, Translation="0,0,0", Rotation="0,0,0", Scale="0.8,0.8,0.8") {
    rotationMode=BMM_Additive
    rotationSpace=BCS_BoneSpace
    scaleMode=BMM_Additive
    scaleSpace=BCS_BoneSpace
    translationMode=BMM_Additive
    translationSpace=BCS_BoneSpace
    alphaBoolBlend {blendOption=Linear}
    alphaScaleBiasClamp {bInterpResult=true, interpSpeedDecreasing=20, interpSpeedIncreasing=20}
    boneToModify {boneName=dyn_hair_c_main_02}
    scale=<0.8 0.8 0.8>
   }
   n30 = "Transform (Modify) Bone - Bone: dyn_hair_c_main_03"(ComponentPose=@n29, Alpha=1, Translation="0,0,0", Rotation="0,0,0", Scale="0.7,0.7,0.7") {
    rotationMode=BMM_Additive
    rotationSpace=BCS_BoneSpace
    scaleMode=BMM_Additive
    scaleSpace=BCS_BoneSpace
    translationMode=BMM_Additive
    translationSpace=BCS_BoneSpace
    alphaBoolBlend {blendOption=Linear}
    alphaScaleBiasClamp {bInterpResult=true, interpSpeedDecreasing=20, interpSpeedIncreasing=20}
    boneToModify {boneName=dyn_hair_c_main_03}
    scale=<0.7 0.7 0.7>
   }
   n31 = "Transform (Modify) Bone - Bone: dyn_hair_c_main_04"(ComponentPose=@n30, Alpha=1, Translation="0,0,0", Rotation="0,0,0", Scale="1,1,1") {
    rotationMode=BMM_Additive
    rotationSpace=BCS_BoneSpace
    scaleMode=BMM_Additive
    scaleSpace=BCS_BoneSpace
    translationMode=BMM_Additive
    translationSpace=BCS_BoneSpace
    alphaBoolBlend {blendOption=Linear}
    alphaScaleBiasClamp {bInterpResult=true, interpSpeedDecreasing=20, interpSpeedIncreasing=20}
    boneToModify {boneName=dyn_hair_c_main_04}
   }
   n32 = "Transform (Modify) Bone - Bone: dyn_hair_l_main_c_02"(ComponentPose=@n31, Alpha=1, Translation="0,0,0", Rotation="0,0,0", Scale="0.8,0.8,0.8") {
    rotationMode=BMM_Additive
    rotationSpace=BCS_BoneSpace
    scaleMode=BMM_Additive
    scaleSpace=BCS_BoneSpace
    translationMode=BMM_Additive
    translationSpace=BCS_BoneSpace
    alphaBoolBlend {blendOption=Linear}
    alphaScaleBiasClamp {bInterpResult=true, interpSpeedDecreasing=20, interpSpeedIncreasing=20}
    boneToModify {boneName=dyn_hair_l_main_c_02}
    scale=<0.8 0.8 0.8>
   }
   n33 = "Transform (Modify) Bone - Bone: dyn_hair_l_main_c_03"(ComponentPose=@n32, Alpha=1, Translation="0,0,0", Rotation="0,0,0", Scale="0.7,0.7,0.7") {
    rotationMode=BMM_Additive
    rotationSpace=BCS_BoneSpace
    scaleMode=BMM_Additive
    scaleSpace=BCS_BoneSpace
    translationMode=BMM_Additive
    translationSpace=BCS_BoneSpace
    alphaBoolBlend {blendOption=Linear}
    alphaScaleBiasClamp {bInterpResult=true, interpSpeedDecreasing=20, interpSpeedIncreasing=20}
    boneToModify {boneName=dyn_hair_l_main_c_03}
    scale=<0.7 0.7 0.7>
   }
   n34 = Component To Local(ComponentPose=@n33)
   n35 = Use cached pose 'PostSim'()
   n36 = Get isWearingCape() -> (isWearingCape:bool)
   n37 = "Constraint - dyn_hair_bang_a_02 (RRRRR)"!(
    ComponentPose=Local To Component(LocalPose=Blend Poses by bool(BlendPose_0=@n34, BlendPose_1=@n35, bActiveValue=@n36) {blendType=Linear})
    Alpha=1
    ConstraintWeights_0=0.375
    ConstraintWeights_1:float=0.0193
    ConstraintWeights_2:float=0.3652
    ConstraintWeights_3:float=0.0008
    ConstraintWeights_4:float=0.2396
   ) -> (Pose) {
    alphaBoolBlend {blendOption=Linear}
    boneToModify {boneName=dyn_hair_bang_a_02}
    constraintSetup [
     {transformType=Rotation, targetBone {boneName=head}}
     {transformType=Rotation, targetBone {boneName=dyn_hair_l_main_MSTR}}
     {transformType=Rotation, targetBone {boneName=dyn_hair_l_main}}
     {transformType=Rotation, targetBone {boneName=dyn_hair_l_main_b_02}}
     {transformType=Rotation, targetBone {boneName=dyn_hair_l_main_a_01}}
    ]
    constraintWeights [1,1,1,1,1]
   }
   n38 = "Constraint - dyn_hair_bang_a_03 (PPPPP)"!(
    ComponentPose=@n37
    Alpha=1
    ConstraintWeights_0=0.0048
    ConstraintWeights_1:float=0.0594
    ConstraintWeights_2:float=0.657
    ConstraintWeights_3:float=0.2605
    ConstraintWeights_4:float=0.0183
   ) -> (Pose) {
    alphaBoolBlend {blendOption=Linear}
    boneToModify {boneName=dyn_hair_bang_a_03}
    constraintSetup [
     {targetBone {boneName=head}}
     {targetBone {boneName=dyn_hair_l_main}}
     {targetBone {boneName=dyn_hair_l_main_a_01}}
     {targetBone {boneName=dyn_hair_l_main_a_02}}
     {targetBone {boneName=dyn_hair_l_main_a_03}}
    ]
    constraintWeights [1,1,1,1,1]
   }
   n39 = "Transform (Modify) Bone - Bone: dyn_hair_bang_b_MSTR"(ComponentPose=@n38, Alpha=1, Translation="0,0,0", Rotation="-8.250,0.263,-0.058", Scale="0.8,0.8,0.8") {
    rotationMode=BMM_Additive
    rotationSpace=BCS_BoneSpace
    scaleMode=BMM_Additive
    scaleSpace=BCS_BoneSpace
    translationMode=BMM_Additive
    translationSpace=BCS_BoneSpace
    alphaBoolBlend {blendOption=Linear}
    alphaScaleBiasClamp {bInterpResult=true, interpSpeedDecreasing=20, interpSpeedIncreasing=20}
    boneToModify {boneName=dyn_hair_bang_b_MSTR}
    rotation=<rot -8.25 0.263 -0.058>
    scale=<0.8 0.8 0.8>
   }
   Output Pose(Result=Component To Local(ComponentPose=@n39)) {layerGroup=None}
   n40 = "Transform (Modify) Bone - Bone: dyn_root_simspace"(
    ComponentPose=Local To Component(LocalPose=Copy Pose From Mesh() {bUseAttachedParent=true})
    Alpha=1
    Translation="0,0,0"
    Rotation=Get sim_space_rotation() -> (sim_space_rotation:Rotator)
    Scale="1,1,1"
   ) {
    rotationMode=BMM_Replace
    rotationSpace=BCS_WorldSpace
    translationSpace=BCS_BoneSpace
    alphaBoolBlend {blendOption=Linear}
    alphaScaleBiasClamp {bInterpResult=true, interpSpeedDecreasing=2, interpSpeedIncreasing=2}
    boneToModify {boneName=dyn_root_simspace}
   }
   n41 = Anim Dynamics - Bone: dyn_hair_l_main!(ComponentPose=@n40, Alpha=0, GravityScale:float=1, GravityOverride:Vector="0,0,0", ExternalForce="0,0,0") -> (Pose) {
    angularDampingOverride=1
    angularSpringConstant=75
    bAngularSpring=true
    bGravityOverrideInSimSpace=true
    bOverrideAngularDamping=true
    simulationSpace=BoneRelative
    alphaBoolBlend {blendOption=Linear}
    boundBone {boneName=dyn_hair_l_main}
    componentAppliedLinearAccClamp=<5000 5000 5000>
    componentLinearAccScale=<1.2 1.2 1.3>
    componentLinearVelScale=<1.2 1.2 1.3>
    physicsBodyDefinitions [{
      sphereCollisionRadius=0
      boundBone {boneName=dyn_hair_l_main}
      boxExtents=<10 20 20>
      constraintSetup {angularLimitsMax=<45 -10 45>, angularLimitsMin=<-45 -45 -45>, angularTarget=<0 0 0>}
      localJointOffset=<2 0 -0.5>
     }]
    relativeSpaceBone {boneName=dyn_root_simspace}
   }
   n42 = Anim Dynamics - Bone: dyn_hair_l_main_c_01!(ComponentPose=@n41, Alpha=0, GravityScale:float=1, GravityOverride:Vector="0,0,0", ExternalForce="0,0,0") -> (Pose) {
    angularDampingOverride=1
    angularSpringConstant=75
    bAngularSpring=true
    bGravityOverrideInSimSpace=true
    bOverrideAngularDamping=true
    simulationSpace=BoneRelative
    alphaBoolBlend {blendOption=Linear}
    boundBone {boneName=dyn_hair_l_main_c_01}
    componentAppliedLinearAccClamp=<5000 5000 5000>
    componentLinearAccScale=<1.2 1.2 1.3>
    componentLinearVelScale=<1.2 1.2 1.3>
    physicsBodyDefinitions [{
      sphereCollisionRadius=0
      boundBone {boneName=dyn_hair_l_main_c_01}
      boxExtents=<10 20 20>
      constraintSetup {angularLimitsMax=<45 -10 45>, angularLimitsMin=<-45 -45 -45>, angularTarget=<0 0 0>}
      localJointOffset=<2 0 -0.5>
     }]
    relativeSpaceBone {boneName=dyn_root_simspace}
   }
   n43 = Anim Dynamics - Bone: dyn_hair_l_main_b_02!(ComponentPose=@n42, Alpha=0.5, GravityScale:float=1, GravityOverride:Vector="0,0,0", ExternalForce="0,0,0") -> (Pose) {
    angularDampingOverride=1
    angularSpringConstant=750
    bAngularSpring=true
    bGravityOverrideInSimSpace=true
    bOverrideAngularDamping=true
    simulationSpace=BoneRelative
    alphaBoolBlend {blendOption=Linear}
    boundBone {boneName=dyn_hair_l_main_b_02}
    componentAppliedLinearAccClamp=<5000 5000 5000>
    componentLinearAccScale=<1.2 1.2 1.3>
    componentLinearVelScale=<1.2 1.2 1.3>
    physicsBodyDefinitions [{
      sphereCollisionRadius=0
      boundBone {boneName=dyn_hair_l_main_b_02}
      boxExtents=<10 20 20>
      constraintSetup {angularLimitsMax=<45 -10 15>, angularLimitsMin=<-45 -45 -15>, angularTarget=<0 0 0>}
      localJointOffset=<2 0 -0.5>
     }]
    relativeSpaceBone {boneName=dyn_root_simspace}
   }
   n44 = Anim Dynamics - Bone: dyn_hair_l_main_b_03!(ComponentPose=@n43, Alpha=0.4, GravityScale:float=1, GravityOverride:Vector="0,0,0", ExternalForce="0,0,0") -> (Pose) {
    angularDampingOverride=1
    angularSpringConstant=750
    bAngularSpring=true
    bGravityOverrideInSimSpace=true
    bOverrideAngularDamping=true
    simulationSpace=BoneRelative
    alphaBoolBlend {blendOption=Linear}
    boundBone {boneName=dyn_hair_l_main_b_03}
    componentAppliedLinearAccClamp=<5000 5000 5000>
    componentLinearAccScale=<1.2 1.2 1.3>
    componentLinearVelScale=<1.2 1.2 1.3>
    physicsBodyDefinitions [{
      sphereCollisionRadius=0
      boundBone {boneName=dyn_hair_l_main_b_03}
      boxExtents=<10 20 20>
      constraintSetup {angularLimitsMax=<45 -10 15>, angularLimitsMin=<-45 -45 -15>, angularTarget=<0 0 0>}
      localJointOffset=<2 0 -0.5>
     }]
    relativeSpaceBone {boneName=dyn_root_simspace}
   }
   n45 = Anim Dynamics - Bone: dyn_hair_l_main_b_04!(ComponentPose=@n44, Alpha=0.3, GravityScale:float=1, GravityOverride:Vector="0,0,0", ExternalForce="0,0,0") -> (Pose) {
    angularDampingOverride=1
    angularSpringConstant=750
    bAngularSpring=true
    bGravityOverrideInSimSpace=true
    bOverrideAngularDamping=true
    simulationSpace=BoneRelative
    alphaBoolBlend {blendOption=Linear}
    boundBone {boneName=dyn_hair_l_main_b_04}
    componentAppliedLinearAccClamp=<5000 5000 5000>
    componentLinearAccScale=<1.2 1.2 1.3>
    componentLinearVelScale=<1.2 1.2 1.3>
    physicsBodyDefinitions [{
      sphereCollisionRadius=0
      boundBone {boneName=dyn_hair_l_main_b_04}
      boxExtents=<10 20 20>
      constraintSetup {angularLimitsMax=<45 -10 15>, angularLimitsMin=<-45 -45 -15>, angularTarget=<0 0 0>}
      localJointOffset=<2 0 -0.5>
     }]
    relativeSpaceBone {boneName=dyn_root_simspace}
   }
   n46 = Anim Dynamics - Bone: dyn_hair_l_main_b_05!(ComponentPose=@n45, Alpha=0.3, GravityScale:float=1, GravityOverride:Vector="0,0,0", ExternalForce="0,0,0") -> (Pose) {
    angularDampingOverride=1
    angularSpringConstant=750
    bAngularSpring=true
    bGravityOverrideInSimSpace=true
    bOverrideAngularDamping=true
    simulationSpace=BoneRelative
    alphaBoolBlend {blendOption=Linear}
    boundBone {boneName=dyn_hair_l_main_b_05}
    componentAppliedLinearAccClamp=<5000 5000 5000>
    componentLinearAccScale=<1.2 1.2 1.3>
    componentLinearVelScale=<1.2 1.2 1.3>
    physicsBodyDefinitions [{
      sphereCollisionRadius=0
      boundBone {boneName=dyn_hair_l_main_b_05}
      boxExtents=<10 20 20>
      constraintSetup {angularLimitsMax=<45 -10 15>, angularLimitsMin=<-45 -45 -15>, angularTarget=<0 0 0>}
      localJointOffset=<2 0 -0.5>
     }]
    relativeSpaceBone {boneName=dyn_root_simspace}
   }
   n47 = Anim Dynamics - Bone: dyn_hair_l_main_a_01!(ComponentPose=@n46, Alpha=1, GravityScale:float=1, GravityOverride:Vector="0,0,0", ExternalForce="0,0,0") -> (Pose) {
    angularDampingOverride=1
    angularSpringConstant=75
    bAngularSpring=true
    bGravityOverrideInSimSpace=true
    bOverrideAngularDamping=true
    simulationSpace=BoneRelative
    alphaBoolBlend {blendOption=Linear}
    boundBone {boneName=dyn_hair_l_main_a_01}
    componentAppliedLinearAccClamp=<5000 5000 5000>
    componentLinearAccScale=<1.2 1.2 1.3>
    componentLinearVelScale=<1.2 1.2 1.3>
    physicsBodyDefinitions [{
      sphereCollisionRadius=0
      boundBone {boneName=dyn_hair_l_main_a_01}
      boxExtents=<10 20 20>
      constraintSetup {angularLimitsMax=<45 5 45>, angularLimitsMin=<-45 -45 -45>, angularTarget=<0 0 0>}
      localJointOffset=<2 0 0>
     }]
    relativeSpaceBone {boneName=dyn_root_simspace}
   }
   n48 = Anim Dynamics - Bone: dyn_hair_l_main_a_02!(ComponentPose=@n47, Alpha=1, GravityScale:float=1, GravityOverride:Vector="0,0,0", ExternalForce="0,0,0") -> (Pose) {
    angularDampingOverride=1
    angularSpringConstant=75
    bAngularSpring=true
    bGravityOverrideInSimSpace=true
    bOverrideAngularDamping=true
    simulationSpace=BoneRelative
    alphaBoolBlend {blendOption=Linear}
    boundBone {boneName=dyn_hair_l_main_a_02}
    componentAppliedLinearAccClamp=<5000 5000 5000>
    componentLinearAccScale=<1.2 1.2 1.3>
    componentLinearVelScale=<1.2 1.2 1.3>
    physicsBodyDefinitions [{
      sphereCollisionRadius=0
      boundBone {boneName=dyn_hair_l_main_a_02}
      boxExtents=<10 20 20>
      constraintSetup {angularLimitsMax=<45 5 45>, angularLimitsMin=<-45 -45 -45>, angularTarget=<0 0 0>}
      localJointOffset=<2 0 0>
     }]
    relativeSpaceBone {boneName=dyn_root_simspace}
   }
   n49 = Anim Dynamics - Bone: dyn_hair_l_main_a_03!(ComponentPose=@n48, Alpha=1, GravityScale:float=1, GravityOverride:Vector="0,0,0", ExternalForce="0,0,0") -> (Pose) {
    angularDampingOverride=1
    angularSpringConstant=75
    bAngularSpring=true
    bGravityOverrideInSimSpace=true
    bOverrideAngularDamping=true
    bUseSphericalLimits=true
    simulationSpace=BoneRelative
    alphaBoolBlend {blendOption=Linear}
    boundBone {boneName=dyn_hair_l_main_a_03}
    componentAppliedLinearAccClamp=<5000 5000 5000>
    componentLinearAccScale=<1.2 1.2 1.3>
    componentLinearVelScale=<1.2 1.2 1.3>
    physicsBodyDefinitions [{
      sphereCollisionRadius=0
      boundBone {boneName=dyn_hair_l_main_a_03}
      boxExtents=<10 20 20>
      constraintSetup {angularLimitsMax=<45 5 45>, angularLimitsMin=<-45 -45 -45>, angularTarget=<0 0 0>}
      localJointOffset=<2 0 0>
     }]
    relativeSpaceBone {boneName=dyn_root_simspace}
    sphericalLimits [{limitRadius=7.5, drivingBone {boneName=clavicle_l}, sphereLocalOffset=<9 0 0>}]
   }
   n50 = Anim Dynamics - Bone: dyn_hair_l_main_a_04!(ComponentPose=@n49, Alpha=1, GravityScale:float=1, GravityOverride:Vector="0,0,0", ExternalForce="0,0,0") -> (Pose) {
    angularDampingOverride=1
    angularSpringConstant=75
    bAngularSpring=true
    bGravityOverrideInSimSpace=true
    bOverrideAngularDamping=true
    simulationSpace=BoneRelative
    alphaBoolBlend {blendOption=Linear}
    boundBone {boneName=dyn_hair_l_main_a_04}
    componentAppliedLinearAccClamp=<5000 5000 5000>
    componentLinearAccScale=<1.2 1.2 1.3>
    componentLinearVelScale=<1.2 1.2 1.3>
    physicsBodyDefinitions [{
      sphereCollisionRadius=0
      boundBone {boneName=dyn_hair_l_main_a_04}
      boxExtents=<10 20 20>
      constraintSetup {angularLimitsMax=<45 5 45>, angularLimitsMin=<-45 -45 -45>, angularTarget=<0 0 0>}
      localJointOffset=<2 0 0>
     }]
    relativeSpaceBone {boneName=dyn_root_simspace}
   }
   n51 = Anim Dynamics - Bone: dyn_hair_l_main_a_05!(ComponentPose=@n50, Alpha=1, GravityScale:float=1, GravityOverride:Vector="0,0,0", ExternalForce="0,0,0") -> (Pose) {
    angularDampingOverride=1
    angularSpringConstant=75
    bAngularSpring=true
    bGravityOverrideInSimSpace=true
    bOverrideAngularDamping=true
    simulationSpace=BoneRelative
    alphaBoolBlend {blendOption=Linear}
    boundBone {boneName=dyn_hair_l_main_a_05}
    componentAppliedLinearAccClamp=<5000 5000 5000>
    componentLinearAccScale=<1.2 1.2 1.3>
    componentLinearVelScale=<1.2 1.2 1.3>
    physicsBodyDefinitions [{
      sphereCollisionRadius=0
      boundBone {boneName=dyn_hair_l_main_a_05}
      boxExtents=<10 20 20>
      constraintSetup {angularLimitsMax=<45 5 45>, angularLimitsMin=<-45 -45 -45>, angularTarget=<0 0 0>}
      localJointOffset=<2 0 0>
     }]
    relativeSpaceBone {boneName=dyn_root_simspace}
   }
   n52 = Anim Dynamics - Bone: dyn_hair_l_main_f_01!(ComponentPose=@n51, Alpha=1, GravityScale:float=1, GravityOverride:Vector="0,0,-980", ExternalForce="0,0,0") -> (Pose) {
    angularDampingOverride=1
    angularSpringConstant=75
    bAngularSpring=true
    bGravityOverrideInSimSpace=true
    bOverrideAngularDamping=true
    simulationSpace=BoneRelative
    alphaBoolBlend {blendOption=Linear}
    boundBone {boneName=dyn_hair_l_main_f_01}
    componentAppliedLinearAccClamp=<5000 5000 5000>
    componentLinearAccScale=<1.1 1.1 1.1>
    componentLinearVelScale=<1.1 1.1 1.1>
    gravityOverride {z=-980}
    physicsBodyDefinitions [{
      sphereCollisionRadius=0
      boundBone {boneName=dyn_hair_l_main_f_01}
      boxExtents=<20 20 10>
      constraintSetup {angularLimitsMax=<45 -10 5>, angularLimitsMin=<-45 -45 -5>, angularTarget=<0 0 0>}
      localJointOffset=<2 0 -0.5>
     }]
    relativeSpaceBone {boneName=dyn_root_simspace}
   }
   n53 = Anim Dynamics - Bone: dyn_hair_l_main_f_02!(ComponentPose=@n52, Alpha=1, GravityScale:float=1, GravityOverride:Vector="0,0,-980", ExternalForce="0,0,0") -> (Pose) {
    angularDampingOverride=1
    angularSpringConstant=75
    bAngularSpring=true
    bGravityOverrideInSimSpace=true
    bOverrideAngularDamping=true
    simulationSpace=BoneRelative
    alphaBoolBlend {blendOption=Linear}
    boundBone {boneName=dyn_hair_l_main_f_02}
    componentAppliedLinearAccClamp=<5000 5000 5000>
    componentLinearAccScale=<1.1 1.1 1.1>
    componentLinearVelScale=<1.1 1.1 1.1>
    gravityOverride {z=-980}
    physicsBodyDefinitions [{
      sphereCollisionRadius=0
      boundBone {boneName=dyn_hair_l_main_f_02}
      boxExtents=<20 20 10>
      constraintSetup {angularLimitsMax=<45 -10 5>, angularLimitsMin=<-45 -45 -5>, angularTarget=<0 0 0>}
      localJointOffset=<2 0 -0.5>
     }]
    relativeSpaceBone {boneName=dyn_root_simspace}
   }
   n54 = Anim Dynamics - Bone: dyn_hair_r_main_08!(ComponentPose=@n53, Alpha=1, GravityScale:float=0, GravityOverride:Vector="0,0,-980", ExternalForce="0,0,0") -> (Pose) {
    angularDampingOverride=1
    angularSpringConstant=75
    bAngularSpring=true
    bGravityOverrideInSimSpace=true
    bOverrideAngularDamping=true
    simulationSpace=BoneRelative
    alphaBoolBlend {blendOption=Linear}
    boundBone {boneName=dyn_hair_r_main_08}
    componentAppliedLinearAccClamp=<5000 5000 5000>
    componentLinearAccScale=<1.2 1.2 1.1>
    componentLinearVelScale=<1.2 1.2 1.1>
    gravityOverride {z=-980}
    physicsBodyDefinitions [{
      sphereCollisionRadius=0
      boundBone {boneName=dyn_hair_r_main_08}
      boxExtents=<20 20 10>
      constraintSetup {angularLimitsMax=<45 -10 5>, angularLimitsMin=<-45 -45 -5>, angularTarget=<0 0 0>}
      localJointOffset=<2 0 -0.5>
     }]
    relativeSpaceBone {boneName=dyn_root_simspace}
   }
   n55 = Anim Dynamics - Bone: dyn_hair_l_main_c_02!(ComponentPose=@n54, Alpha=1, GravityScale:float=1, GravityOverride:Vector="0,0,-980", ExternalForce="0,0,0") -> (Pose) {
    angularDampingOverride=1
    angularSpringConstant=75
    bAngularSpring=true
    bGravityOverrideInSimSpace=true
    bOverrideAngularDamping=true
    simulationSpace=BoneRelative
    alphaBoolBlend {blendOption=Linear}
    boundBone {boneName=dyn_hair_l_main_c_02}
    componentAppliedLinearAccClamp=<10000 10000 8000>
    componentLinearAccScale=<1.4 1.4 1.6>
    componentLinearVelScale=<1.4 1.4 1.6>
    gravityOverride {z=-980}
    physicsBodyDefinitions [{
      sphereCollisionRadius=0
      boundBone {boneName=dyn_hair_l_main_c_02}
      boxExtents=<20 20 10>
      constraintSetup {angularLimitsMax=<45 -10 45>, angularLimitsMin=<-45 -45 -45>, angularTarget=<0 0 0>}
      localJointOffset=<2 0 -0.5>
     }]
    relativeSpaceBone {boneName=dyn_root_simspace}
   }
   n56 = Anim Dynamics - Bone: dyn_hair_l_main_c_03!(ComponentPose=@n55, Alpha=1, GravityScale:float=1, GravityOverride:Vector="0,0,-980", ExternalForce="0,0,0") -> (Pose) {
    angularDampingOverride=1
    angularSpringConstant=75
    bAngularSpring=true
    bGravityOverrideInSimSpace=true
    bOverrideAngularDamping=true
    simulationSpace=BoneRelative
    alphaBoolBlend {blendOption=Linear}
    boundBone {boneName=dyn_hair_l_main_c_03}
    componentAppliedLinearAccClamp=<10000 10000 8000>
    componentLinearAccScale=<1.4 1.4 1.6>
    componentLinearVelScale=<1.4 1.4 1.6>
    gravityOverride {z=-980}
    physicsBodyDefinitions [{
      sphereCollisionRadius=0
      boundBone {boneName=dyn_hair_l_main_c_03}
      boxExtents=<20 20 10>
      constraintSetup {angularLimitsMax=<45 -10 45>, angularLimitsMin=<-45 -45 -45>, angularTarget=<0 0 0>}
      localJointOffset=<2 0 -0.5>
     }]
    relativeSpaceBone {boneName=dyn_root_simspace}
   }
   n57 = Anim Dynamics - Bone: dyn_hair_l_main_c_04!(ComponentPose=@n56, Alpha=1, GravityScale:float=1, GravityOverride:Vector="0,0,-980", ExternalForce="0,0,0") -> (Pose) {
    angularDampingOverride=1
    angularSpringConstant=75
    bAngularSpring=true
    bGravityOverrideInSimSpace=true
    bOverrideAngularDamping=true
    simulationSpace=BoneRelative
    alphaBoolBlend {blendOption=Linear}
    boundBone {boneName=dyn_hair_l_main_c_04}
    componentAppliedLinearAccClamp=<10000 10000 8000>
    componentLinearAccScale=<1.4 1.4 1.6>
    componentLinearVelScale=<1.4 1.4 1.6>
    gravityOverride {z=-980}
    physicsBodyDefinitions [{
      sphereCollisionRadius=0
      boundBone {boneName=dyn_hair_l_main_c_04}
      boxExtents=<20 20 10>
      constraintSetup {angularLimitsMax=<45 -10 45>, angularLimitsMin=<-45 -45 -45>, angularTarget=<0 0 0>}
      localJointOffset=<2 0 -0.5>
     }]
    relativeSpaceBone {boneName=dyn_root_simspace}
   }
   n58 = Anim Dynamics - Bone: dyn_hair_l_main_c_05!(ComponentPose=@n57, Alpha=1, GravityScale:float=1, GravityOverride:Vector="0,0,-980", ExternalForce="0,0,0") -> (Pose) {
    angularDampingOverride=1
    angularSpringConstant=75
    bAngularSpring=true
    bGravityOverrideInSimSpace=true
    bOverrideAngularDamping=true
    simulationSpace=BoneRelative
    alphaBoolBlend {blendOption=Linear}
    boundBone {boneName=dyn_hair_l_main_c_05}
    componentAppliedLinearAccClamp=<10000 10000 8000>
    componentLinearAccScale=<1.4 1.4 1.6>
    componentLinearVelScale=<1.4 1.4 1.6>
    gravityOverride {z=-980}
    physicsBodyDefinitions [{
      sphereCollisionRadius=0
      boundBone {boneName=dyn_hair_l_main_c_05}
      boxExtents=<20 20 10>
      constraintSetup {angularLimitsMax=<45 -10 45>, angularLimitsMin=<-45 -45 -45>, angularTarget=<0 0 0>}
      localJointOffset=<2 0 -0.5>
     }]
    relativeSpaceBone {boneName=dyn_root_simspace}
   }
   n59 = Anim Dynamics - Bone: dyn_hair_l_main_d_01!(ComponentPose=@n58, Alpha=1, GravityScale:float=1, GravityOverride:Vector="0,0,-980", ExternalForce="0,0,0") -> (Pose) {
    angularDampingOverride=1
    angularSpringConstant=75
    bAngularSpring=true
    bGravityOverrideInSimSpace=true
    bOverrideAngularDamping=true
    simulationSpace=BoneRelative
    alphaBoolBlend {blendOption=Linear}
    boundBone {boneName=dyn_hair_l_main_d_01}
    componentAppliedLinearAccClamp=<5000 5000 4000>
    componentLinearAccScale=<1.2 1.2 1.4>
    componentLinearVelScale=<1.2 1.2 1.4>
    gravityOverride {z=-980}
    physicsBodyDefinitions [{
      sphereCollisionRadius=0
      boundBone {boneName=dyn_hair_l_main_d_01}
      boxExtents=<20 20 10>
      constraintSetup {angularLimitsMax=<45 -10 45>, angularLimitsMin=<-45 -45 -45>, angularTarget=<0 0 0>}
      localJointOffset=<2 0 -0.5>
     }]
    relativeSpaceBone {boneName=dyn_root_simspace}
   }
   n60 = Anim Dynamics - Bone: dyn_hair_l_main_d_02!(ComponentPose=@n59, Alpha=1, GravityScale:float=1, GravityOverride:Vector="0,0,-980", ExternalForce="0,0,0") -> (Pose) {
    angularDampingOverride=1
    angularSpringConstant=75
    bAngularSpring=true
    bGravityOverrideInSimSpace=true
    bOverrideAngularDamping=true
    simulationSpace=BoneRelative
    alphaBoolBlend {blendOption=Linear}
    boundBone {boneName=dyn_hair_l_main_d_02}
    componentAppliedLinearAccClamp=<5000 5000 4000>
    componentLinearAccScale=<1.2 1.2 1.4>
    componentLinearVelScale=<1.2 1.2 1.4>
    gravityOverride {z=-980}
    physicsBodyDefinitions [{
      sphereCollisionRadius=0
      boundBone {boneName=dyn_hair_l_main_d_02}
      boxExtents=<20 20 10>
      constraintSetup {angularLimitsMax=<45 -10 45>, angularLimitsMin=<-45 -45 -45>, angularTarget=<0 0 0>}
      localJointOffset=<2 0 -0.5>
     }]
    relativeSpaceBone {boneName=dyn_root_simspace}
   }
   n61 = Anim Dynamics - Bone: dyn_hair_c_main_01!(ComponentPose=@n60, Alpha=1, GravityScale:float=1, GravityOverride:Vector="0,0,-980", ExternalForce="0,0,0") -> (Pose) {
    angularDampingOverride=1
    angularSpringConstant=75
    bAngularSpring=true
    bGravityOverrideInSimSpace=true
    bOverrideAngularDamping=true
    simulationSpace=BoneRelative
    alphaBoolBlend {blendOption=Linear}
    boundBone {boneName=dyn_hair_c_main_01}
    componentAppliedLinearAccClamp=<5000 5000 4000>
    componentLinearAccScale=<1.2 1.2 1.4>
    componentLinearVelScale=<1.2 1.2 1.4>
    gravityOverride {z=-980}
    physicsBodyDefinitions [{
      sphereCollisionRadius=0
      boundBone {boneName=dyn_hair_c_main_01}
      boxExtents=<20 20 10>
      constraintSetup {angularLimitsMax=<45 0 45>, angularLimitsMin=<-45 -45 -45>, angularTarget=<0 0 0>}
      localJointOffset=<2 0 -0.5>
     }]
    relativeSpaceBone {boneName=dyn_root_simspace}
   }
   n62 = Anim Dynamics - Bone: dyn_hair_c_main_02!(ComponentPose=@n61, Alpha=1, GravityScale:float=1, GravityOverride:Vector="0,0,-980", ExternalForce="0,0,0") -> (Pose) {
    angularDampingOverride=1
    angularSpringConstant=75
    bAngularSpring=true
    bGravityOverrideInSimSpace=true
    bOverrideAngularDamping=true
    simulationSpace=BoneRelative
    alphaBoolBlend {blendOption=Linear}
    boundBone {boneName=dyn_hair_c_main_02}
    componentAppliedLinearAccClamp=<5000 5000 4000>
    componentLinearAccScale=<1.2 1.2 1.4>
    componentLinearVelScale=<1.2 1.2 1.4>
    gravityOverride {z=-980}
    physicsBodyDefinitions [{
      sphereCollisionRadius=0
      boundBone {boneName=dyn_hair_c_main_02}
      boxExtents=<20 20 10>
      constraintSetup {angularLimitsMax=<45 0 45>, angularLimitsMin=<-45 -45 -45>, angularTarget=<0 0 0>}
      localJointOffset=<2 0 -0.5>
     }]
    relativeSpaceBone {boneName=dyn_root_simspace}
   }
   n63 = Anim Dynamics - Bone: dyn_hair_c_main_03!(ComponentPose=@n62, Alpha=1, GravityScale:float=1, GravityOverride:Vector="0,0,-980", ExternalForce="0,0,0") -> (Pose) {
    angularDampingOverride=1
    bAngularSpring=true
    bGravityOverrideInSimSpace=true
    bOverrideAngularDamping=true
    simulationSpace=BoneRelative
    alphaBoolBlend {blendOption=Linear}
    boundBone {boneName=dyn_hair_c_main_03}
    componentAppliedLinearAccClamp=<5000 5000 5000>
    componentLinearAccScale=<1.2 1.2 1.5>
    componentLinearVelScale=<1.2 1.2 1.5>
    gravityOverride {z=-980}
    physicsBodyDefinitions [{
      sphereCollisionRadius=0
      boundBone {boneName=dyn_hair_c_main_03}
      boxExtents=<20 20 10>
      constraintSetup {angularLimitsMax=<45 0 45>, angularLimitsMin=<-45 -45 -45>, angularTarget=<0 0 0>}
      localJointOffset=<2 0 -0.5>
     }]
    relativeSpaceBone {boneName=dyn_root_simspace}
   }
   n64 = Anim Dynamics - Bone: dyn_hair_c_main_04!(ComponentPose=@n63, Alpha=1, GravityScale:float=1, GravityOverride:Vector="0,0,-980", ExternalForce="0,0,0") -> (Pose) {
    angularDampingOverride=1
    bAngularSpring=true
    bGravityOverrideInSimSpace=true
    bOverrideAngularDamping=true
    simulationSpace=BoneRelative
    alphaBoolBlend {blendOption=Linear}
    boundBone {boneName=dyn_hair_c_main_04}
    componentAppliedLinearAccClamp=<5000 5000 5000>
    componentLinearAccScale=<1.2 1.2 1.5>
    componentLinearVelScale=<1.2 1.2 1.5>
    gravityOverride {z=-980}
    physicsBodyDefinitions [{
      sphereCollisionRadius=0
      boundBone {boneName=dyn_hair_c_main_04}
      boxExtents=<20 20 10>
      constraintSetup {angularLimitsMax=<45 -8 45>, angularLimitsMin=<-45 -45 -45>, angularTarget=<0 0 0>}
      localJointOffset=<2 0 -0.5>
     }]
    relativeSpaceBone {boneName=dyn_root_simspace}
   }
   n65 = Anim Dynamics - Bone: dyn_hair_c_main_05!(ComponentPose=@n64, Alpha=1, GravityScale:float=1, GravityOverride:Vector="0,0,-980", ExternalForce="0,0,0") -> (Pose) {
    angularDampingOverride=1
    bAngularSpring=true
    bGravityOverrideInSimSpace=true
    bOverrideAngularDamping=true
    simulationSpace=BoneRelative
    alphaBoolBlend {blendOption=Linear}
    boundBone {boneName=dyn_hair_c_main_05}
    componentAppliedLinearAccClamp=<5000 5000 5000>
    componentLinearAccScale=<1.2 1.2 1.5>
    componentLinearVelScale=<1.2 1.2 1.5>
    gravityOverride {z=-980}
    physicsBodyDefinitions [{
      sphereCollisionRadius=0
      boundBone {boneName=dyn_hair_c_main_05}
      boxExtents=<20 20 10>
      constraintSetup {angularLimitsMax=<45 -8 45>, angularLimitsMin=<-45 -45 -45>, angularTarget=<0 0 0>}
      localJointOffset=<2 0 -0.5>
     }]
    relativeSpaceBone {boneName=dyn_root_simspace}
   }
   n66 = "Constraint - dyn_hair_l_main_c_02 (P)"(ComponentPose=@n65, Alpha=0.5, ConstraintWeights_0=1) {
    alphaBoolBlend {blendOption=Linear}
    boneToModify {boneName=dyn_hair_l_main_c_02}
    constraintSetup [{targetBone {boneName=spine_05}}]
    constraintWeights [1]
   }
   n67 = "Constraint - dyn_hair_l_main_c_03 (P)"(ComponentPose=@n66, Alpha=0.45, ConstraintWeights_0=1) {
    alphaBoolBlend {blendOption=Linear}
    boneToModify {boneName=dyn_hair_l_main_c_03}
    constraintSetup [{targetBone {boneName=clavicle_l}}]
    constraintWeights [1]
   }
   n68 = "Constraint - dyn_hair_l_main_c_04 (P)"(ComponentPose=@n67, Alpha=0.4, ConstraintWeights_0=1) {
    alphaBoolBlend {blendOption=Linear}
    boneToModify {boneName=dyn_hair_l_main_c_04}
    constraintSetup [{targetBone {boneName=spine_05}}]
    constraintWeights [1]
   }
   n69 = "Constraint - dyn_hair_l_main_c_05 (P)"(ComponentPose=@n68, Alpha=0.3, ConstraintWeights_0=1) {
    alphaBoolBlend {blendOption=Linear}
    boneToModify {boneName=dyn_hair_l_main_c_05}
    constraintSetup [{targetBone {boneName=spine_05}}]
    constraintWeights [1]
   }
   n70 = "Constraint - dyn_hair_l_main_d_01 (P)"(ComponentPose=@n69, Alpha=0.4, ConstraintWeights_0=1) {
    alphaBoolBlend {blendOption=Linear}
    boneToModify {boneName=dyn_hair_l_main_d_01}
    constraintSetup [{targetBone {boneName=spine_05}}]
    constraintWeights [1]
   }
   n71 = "Constraint - dyn_hair_l_main_d_02 (P)"(ComponentPose=@n70, Alpha=0.1, ConstraintWeights_0=1) {
    alphaBoolBlend {blendOption=Linear}
    boneToModify {boneName=dyn_hair_l_main_d_02}
    constraintSetup [{targetBone {boneName=spine_05}}]
    constraintWeights [1]
   }
   n72 = "Constraint - dyn_hair_c_main_01 (P)"(ComponentPose=@n71, Alpha=0.6, ConstraintWeights_0=1) {
    alphaBoolBlend {blendOption=Linear}
    boneToModify {boneName=dyn_hair_c_main_01}
    constraintSetup [{targetBone {boneName=spine_05}}]
    constraintWeights [1]
   }
   n73 = "Constraint - dyn_hair_c_main_02 (P)"(ComponentPose=@n72, Alpha=0.2, ConstraintWeights_0=1) {
    alphaBoolBlend {blendOption=Linear}
    boneToModify {boneName=dyn_hair_c_main_02}
    constraintSetup [{targetBone {boneName=spine_05}}]
    constraintWeights [1]
   }
   n74 = "Constraint - dyn_hair_c_main_03 (P)"(ComponentPose=@n73, Alpha=0.2, ConstraintWeights_0=1) {
    alphaBoolBlend {blendOption=Linear}
    boneToModify {boneName=dyn_hair_c_main_03}
    constraintSetup [{targetBone {boneName=spine_05}}]
    constraintWeights [1]
   }
   n75 = "Constraint - dyn_hair_c_main_04 (P)"(ComponentPose=@n74, Alpha=0.1, ConstraintWeights_0=1) {
    alphaBoolBlend {blendOption=Linear}
    boneToModify {boneName=dyn_hair_c_main_04}
    constraintSetup [{targetBone {boneName=spine_05}}]
    constraintWeights [1]
   }
   n76 = "Constraint - dyn_hair_c_main_05 (P)"(ComponentPose=@n75, Alpha=0.1, ConstraintWeights_0=1) {
    alphaBoolBlend {blendOption=Linear}
    boneToModify {boneName=dyn_hair_c_main_05}
    constraintSetup [{targetBone {boneName=spine_05}}]
    constraintWeights [1]
   }
   n77 = Anim Dynamics - Bone: dyn_hair_r_main_03!(ComponentPose=@n76, Alpha=1, GravityScale:float=1, GravityOverride:Vector="0,0,-980", ExternalForce="0,0,0") -> (Pose) {
    angularDampingOverride=1
    bAngularSpring=true
    bGravityOverrideInSimSpace=true
    bOverrideAngularDamping=true
    bUseSphericalLimits=true
    simulationSpace=BoneRelative
    alphaBoolBlend {blendOption=Linear}
    boundBone {boneName=dyn_hair_r_main_03}
    componentAppliedLinearAccClamp=<5000 5000 4000>
    componentLinearAccScale=<1.2 1.2 1.5>
    componentLinearVelScale=<1.2 1.2 1.5>
    gravityOverride {z=-980}
    physicsBodyDefinitions [{
      sphereCollisionRadius=0
      boundBone {boneName=dyn_hair_r_main_03}
      boxExtents=<20 20 10>
      constraintSetup {angularLimitsMax=<5 5 45>, angularLimitsMin=<-45 -45 -45>, angularTarget=<0 0 0>}
      localJointOffset=<2 0 -0.5>
     }]
    relativeSpaceBone {boneName=dyn_root_simspace}
    sphericalLimits [
     {limitRadius=15, drivingBone {boneName=None}, sphereLocalOffset=<0 0 0>}
     {limitRadius=12, drivingBone {boneName=None}, sphereLocalOffset=<0 0 0>}
    ]
   }
   n78 = Anim Dynamics - Bone: dyn_hair_r_main_04!(ComponentPose=@n77, Alpha=1, GravityScale:float=1, GravityOverride:Vector="0,0,-980", ExternalForce="0,0,0") -> (Pose) {
    angularDampingOverride=1
    bAngularSpring=true
    bGravityOverrideInSimSpace=true
    bOverrideAngularDamping=true
    bUseSphericalLimits=true
    simulationSpace=BoneRelative
    alphaBoolBlend {blendOption=Linear}
    boundBone {boneName=dyn_hair_r_main_04}
    componentAppliedLinearAccClamp=<5000 5000 6000>
    componentLinearAccScale=<1.2 1.2 1.5>
    componentLinearVelScale=<1.2 1.2 1.5>
    gravityOverride {z=-980}
    physicsBodyDefinitions [{
      sphereCollisionRadius=0
      boundBone {boneName=dyn_hair_r_main_04}
      boxExtents=<20 20 10>
      constraintSetup {angularLimitsMax=<5 20 45>, angularLimitsMin=<-45 -45 -45>, angularTarget=<0 0 0>}
      localJointOffset=<2 0 -0.5>
     }]
    relativeSpaceBone {boneName=dyn_root_simspace}
    sphericalLimits [
     {limitRadius=12, drivingBone {boneName=None}, sphereLocalOffset=<0 0 0>}
     {limitRadius=12, drivingBone {boneName=None}, sphereLocalOffset=<0 0 0>}
    ]
   }
   n79 = Anim Dynamics - Bone: dyn_hair_r_main_05!(ComponentPose=@n78, Alpha=1, GravityScale:float=1, GravityOverride:Vector="0,0,-980", ExternalForce="0,0,0") -> (Pose) {
    angularDampingOverride=1
    bAngularSpring=true
    bGravityOverrideInSimSpace=true
    bOverrideAngularDamping=true
    bUseSphericalLimits=true
    simulationSpace=BoneRelative
    alphaBoolBlend {blendOption=Linear}
    boundBone {boneName=dyn_hair_r_main_05}
    componentAppliedLinearAccClamp=<5000 5000 6000>
    componentLinearAccScale=<1.2 1.2 1.5>
    componentLinearVelScale=<1.2 1.2 1.5>
    gravityOverride {z=-980}
    physicsBodyDefinitions [{
      sphereCollisionRadius=0
      boundBone {boneName=dyn_hair_r_main_05}
      boxExtents=<20 20 10>
      constraintSetup {angularLimitsMax=<5 20 45>, angularLimitsMin=<-45 -45 -45>, angularTarget=<0 0 0>}
      localJointOffset=<2 0 -0.5>
     }]
    relativeSpaceBone {boneName=dyn_root_simspace}
    sphericalLimits [
     {limitRadius=12, drivingBone {boneName=None}, sphereLocalOffset=<0 0 0>}
     {limitRadius=12, drivingBone {boneName=None}, sphereLocalOffset=<0 0 0>}
    ]
   }
   n80 = Anim Dynamics - Bone: dyn_hair_r_main_06!(ComponentPose=@n79, Alpha=1, GravityScale:float=1, GravityOverride:Vector="0,0,-980", ExternalForce="0,0,0") -> (Pose) {
    angularDampingOverride=1
    bAngularSpring=true
    bGravityOverrideInSimSpace=true
    bOverrideAngularDamping=true
    bUseSphericalLimits=true
    simulationSpace=BoneRelative
    alphaBoolBlend {blendOption=Linear}
    boundBone {boneName=dyn_hair_r_main_06}
    componentAppliedLinearAccClamp=<5000 5000 6000>
    componentLinearAccScale=<1.2 1.2 1.5>
    componentLinearVelScale=<1.2 1.2 1.5>
    gravityOverride {z=-980}
    physicsBodyDefinitions [{
      sphereCollisionRadius=0
      boundBone {boneName=dyn_hair_r_main_06}
      boxExtents=<20 20 10>
      constraintSetup {angularLimitsMax=<5 20 45>, angularLimitsMin=<-45 -45 -45>, angularTarget=<0 0 0>}
      localJointOffset=<2 0 -0.5>
     }]
    relativeSpaceBone {boneName=dyn_root_simspace}
    sphericalLimits [
     {limitRadius=12, drivingBone {boneName=None}, sphereLocalOffset=<0 0 0>}
     {limitRadius=12, drivingBone {boneName=None}, sphereLocalOffset=<0 0 0>}
    ]
   }
   n81 = Anim Dynamics - Bone: dyn_hair_r_main_07!(ComponentPose=@n80, Alpha=1, GravityScale:float=1, GravityOverride:Vector="0,0,-980", ExternalForce="0,0,0") -> (Pose) {
    angularDampingOverride=1
    bAngularSpring=true
    bGravityOverrideInSimSpace=true
    bOverrideAngularDamping=true
    bUseSphericalLimits=true
    simulationSpace=BoneRelative
    alphaBoolBlend {blendOption=Linear}
    boundBone {boneName=dyn_hair_r_main_07}
    componentAppliedLinearAccClamp=<5000 5000 6000>
    componentLinearAccScale=<1.2 1.2 1.5>
    componentLinearVelScale=<1.2 1.2 1.5>
    gravityOverride {z=-980}
    physicsBodyDefinitions [{
      sphereCollisionRadius=0
      boundBone {boneName=dyn_hair_r_main_07}
      boxExtents=<20 20 10>
      constraintSetup {angularLimitsMax=<5 20 45>, angularLimitsMin=<-45 -45 -45>, angularTarget=<0 0 0>}
      localJointOffset=<2 0 -0.5>
     }]
    relativeSpaceBone {boneName=dyn_root_simspace}
    sphericalLimits [
     {limitRadius=12, drivingBone {boneName=None}, sphereLocalOffset=<0 0 0>}
     {limitRadius=12, drivingBone {boneName=None}, sphereLocalOffset=<0 0 0>}
    ]
   }
   n82 = Anim Dynamics - Bone: dyn_hair_r_main_01!(ComponentPose=@n81, Alpha=0.5, GravityScale:float=1, GravityOverride:Vector="0,0,0", ExternalForce="0,0,0") -> (Pose) {
    angularDampingOverride=1
    angularSpringConstant=75
    bAngularSpring=true
    bGravityOverrideInSimSpace=true
    bOverrideAngularDamping=true
    bUseSphericalLimits=true
    simulationSpace=BoneRelative
    alphaBoolBlend {blendOption=Linear}
    boundBone {boneName=dyn_hair_r_main_01}
    componentAppliedLinearAccClamp=<5000 5000 5000>
    componentLinearAccScale=<1.2 1.2 1.3>
    componentLinearVelScale=<1.2 1.2 1.3>
    physicsBodyDefinitions [{
      sphereCollisionRadius=0
      boundBone {boneName=dyn_hair_r_main_01}
      boxExtents=<10 20 20>
      constraintSetup {angularLimitsMax=<5 5 5>, angularLimitsMin=<-45 -45 -5>, angularTarget=<0 0 0>}
      localJointOffset=<2 0 -0.5>
     }]
    relativeSpaceBone {boneName=dyn_root_simspace}
    sphericalLimits [
     {limitRadius=12, drivingBone {boneName=None}, sphereLocalOffset=<0 0 0>}
     {limitRadius=12, drivingBone {boneName=None}, sphereLocalOffset=<0 0 0>}
    ]
   }
   n83 = Anim Dynamics - Bone: dyn_hair_r_main_02!(ComponentPose=@n82, Alpha=0.5, GravityScale:float=1, GravityOverride:Vector="0,0,0", ExternalForce="0,0,0") -> (Pose) {
    angularDampingOverride=1
    angularSpringConstant=75
    bAngularSpring=true
    bGravityOverrideInSimSpace=true
    bOverrideAngularDamping=true
    bUseSphericalLimits=true
    simulationSpace=BoneRelative
    alphaBoolBlend {blendOption=Linear}
    boundBone {boneName=dyn_hair_r_main_02}
    componentAppliedLinearAccClamp=<5000 5000 5000>
    componentLinearAccScale=<1.2 1.2 1.3>
    componentLinearVelScale=<1.2 1.2 1.3>
    physicsBodyDefinitions [{
      sphereCollisionRadius=0
      boundBone {boneName=dyn_hair_r_main_02}
      boxExtents=<10 20 20>
      constraintSetup {angularLimitsMax=<5 5 5>, angularLimitsMin=<-45 -45 -5>, angularTarget=<0 0 0>}
      localJointOffset=<2 0 -0.5>
     }]
    relativeSpaceBone {boneName=dyn_root_simspace}
    sphericalLimits [
     {limitRadius=15, drivingBone {boneName=None}, sphereLocalOffset=<0 0 0>}
     {limitRadius=12, drivingBone {boneName=None}, sphereLocalOffset=<0 0 0>}
    ]
   }
   n84 = Anim Dynamics - Bone: dyn_hair_r_main_09!(ComponentPose=@n83, Alpha=1, GravityScale:float=1, GravityOverride:Vector="0,0,0", ExternalForce="0,0,0") -> (Pose) {
    angularDampingOverride=1
    angularSpringConstant=75
    bAngularSpring=true
    bGravityOverrideInSimSpace=true
    bOverrideAngularDamping=true
    bUseSphericalLimits=true
    simulationSpace=BoneRelative
    alphaBoolBlend {blendOption=Linear}
    boundBone {boneName=dyn_hair_r_main_09}
    componentAppliedLinearAccClamp=<5000 5000 5000>
    componentLinearAccScale=<1.2 1.2 1.3>
    componentLinearVelScale=<1.2 1.2 1.3>
    physicsBodyDefinitions [{
      sphereCollisionRadius=0
      boundBone {boneName=dyn_hair_r_main_09}
      boxExtents=<10 20 20>
      constraintSetup {angularLimitsMax=<5 5 20>, angularLimitsMin=<-45 -45 0>, angularTarget=<0 0 0>}
      localJointOffset=<2 0 -0.5>
     }]
    relativeSpaceBone {boneName=dyn_root_simspace}
    sphericalLimits [
     {limitRadius=12, drivingBone {boneName=None}, sphereLocalOffset=<0 0 0>}
     {limitRadius=12, drivingBone {boneName=None}, sphereLocalOffset=<0 0 0>}
    ]
   }
   n85 = Anim Dynamics - Bone: dyn_hair_r_main_10!(ComponentPose=@n84, Alpha=1, GravityScale:float=1, GravityOverride:Vector="0,0,0", ExternalForce="0,0,0") -> (Pose) {
    angularDampingOverride=1
    angularSpringConstant=75
    bAngularSpring=true
    bGravityOverrideInSimSpace=true
    bOverrideAngularDamping=true
    bUseSphericalLimits=true
    simulationSpace=BoneRelative
    alphaBoolBlend {blendOption=Linear}
    boundBone {boneName=dyn_hair_r_main_10}
    componentAppliedLinearAccClamp=<5000 5000 5000>
    componentLinearAccScale=<1.2 1.2 1.3>
    componentLinearVelScale=<1.2 1.2 1.3>
    physicsBodyDefinitions [{
      sphereCollisionRadius=0
      boundBone {boneName=dyn_hair_r_main_10}
      boxExtents=<10 20 20>
      constraintSetup {angularLimitsMax=<5 5 20>, angularLimitsMin=<-45 -45 0>, angularTarget=<0 0 0>}
      localJointOffset=<2 0 -0.5>
     }]
    relativeSpaceBone {boneName=dyn_root_simspace}
    sphericalLimits [
     {limitRadius=12, drivingBone {boneName=None}, sphereLocalOffset=<0 0 0>}
     {limitRadius=12, drivingBone {boneName=None}, sphereLocalOffset=<0 0 0>}
    ]
   }
   n86 = Anim Dynamics - Bone: dyn_hair_bang_b_01!(ComponentPose=@n85, Alpha=0.5, GravityScale:float=1, GravityOverride:Vector="0,0,0", ExternalForce="0,0,0") -> (Pose) {
    angularDampingOverride=0.999
    angularSpringConstant=1250
    bAngularSpring=true
    bGravityOverrideInSimSpace=true
    bOverrideAngularDamping=true
    bUseSphericalLimits=true
    simulationSpace=BoneRelative
    alphaBoolBlend {blendOption=Linear}
    boundBone {boneName=dyn_hair_bang_b_01}
    componentAppliedLinearAccClamp=<5000 5000 5000>
    componentLinearAccScale=<1.2 1.2 1.3>
    componentLinearVelScale=<1.2 1.2 1.3>
    physicsBodyDefinitions [{
      sphereCollisionRadius=0
      boundBone {boneName=dyn_hair_bang_b_01}
      boxExtents=<10 20 20>
      constraintSetup {angularLimitsMax=<5 40 15>, angularLimitsMin=<-45 -45 -40>, angularTarget=<0 0 0>}
      localJointOffset=<2 0 0>
     }]
    relativeSpaceBone {boneName=root}
    sphericalLimits [
     {limitRadius=15, drivingBone {boneName=head}, sphereLocalOffset=<0 0 -8>}
     {limitRadius=12, drivingBone {boneName=None}, sphereLocalOffset=<0 0 0>}
    ]
   }
   n87 = Anim Dynamics - Bone: dyn_hair_bang_b_02!(ComponentPose=@n86, Alpha=0.8, GravityScale:float=1, GravityOverride:Vector="0,0,0", ExternalForce="0,0,0") -> (Pose) {
    angularDampingOverride=0.999
    angularSpringConstant=1000
    bAngularSpring=true
    bGravityOverrideInSimSpace=true
    bOverrideAngularDamping=true
    bUseSphericalLimits=true
    simulationSpace=BoneRelative
    alphaBoolBlend {blendOption=Linear}
    boundBone {boneName=dyn_hair_bang_b_02}
    componentAppliedLinearAccClamp=<5000 5000 5000>
    componentLinearAccScale=<1.2 1.2 1.3>
    componentLinearVelScale=<1.2 1.2 1.3>
    physicsBodyDefinitions [{
      sphereCollisionRadius=0
      boundBone {boneName=dyn_hair_bang_b_02}
      boxExtents=<10 20 20>
      constraintSetup {angularLimitsMax=<5 40 0>, angularLimitsMin=<-45 -45 -40>, angularTarget=<0 0 0>}
      localJointOffset=<2 0 0>
     }]
    relativeSpaceBone {boneName=root}
    sphericalLimits [
     {limitRadius=15, drivingBone {boneName=head}, sphereLocalOffset=<0 0 -8>}
     {limitRadius=12, drivingBone {boneName=None}, sphereLocalOffset=<0 0 0>}
    ]
   }
   n88 = Anim Dynamics - Bone: dyn_hair_bang_b_03!(ComponentPose=@n87, Alpha=1, GravityScale:float=1, GravityOverride:Vector="0,0,0", ExternalForce="0,0,0") -> (Pose) {
    angularDampingOverride=0.999
    angularSpringConstant=750
    bAngularSpring=true
    bGravityOverrideInSimSpace=true
    bOverrideAngularDamping=true
    bUseSphericalLimits=true
    simulationSpace=BoneRelative
    alphaBoolBlend {blendOption=Linear}
    boundBone {boneName=dyn_hair_bang_b_03}
    componentAppliedLinearAccClamp=<5000 5000 5000>
    componentLinearAccScale=<1.2 1.2 1.3>
    componentLinearVelScale=<1.2 1.2 1.3>
    physicsBodyDefinitions [{
      sphereCollisionRadius=0
      boundBone {boneName=dyn_hair_bang_b_03}
      boxExtents=<10 20 20>
      constraintSetup {angularLimitsMax=<5 40 10>, angularLimitsMin=<-45 -45 -40>, angularTarget=<0 0 0>}
      localJointOffset=<2 0 0>
     }]
    relativeSpaceBone {boneName=root}
    sphericalLimits [
     {limitRadius=15, drivingBone {boneName=head}, sphereLocalOffset=<0 0 -8>}
     {limitRadius=12, drivingBone {boneName=None}, sphereLocalOffset=<0 0 0>}
    ]
   }
   n89 = Anim Dynamics - Bone: dyn_hair_bang_b_04!(ComponentPose=@n88, Alpha=0.5, GravityScale:float=1, GravityOverride:Vector="0,0,0", ExternalForce="0,0,0") -> (Pose) {
    angularDampingOverride=0.99
    angularSpringConstant=1500
    bAngularSpring=true
    bGravityOverrideInSimSpace=true
    bOverrideAngularDamping=true
    bUseSphericalLimits=true
    simulationSpace=BoneRelative
    alphaBoolBlend {blendOption=Linear}
    boundBone {boneName=dyn_hair_bang_b_04}
    componentAppliedLinearAccClamp=<5000 5000 5000>
    componentLinearAccScale=<1.2 1.2 1.3>
    componentLinearVelScale=<1.2 1.2 1.3>
    physicsBodyDefinitions [{
      sphereCollisionRadius=0
      boundBone {boneName=dyn_hair_bang_b_04}
      boxExtents=<10 20 20>
      constraintSetup {angularLimitsMax=<5 5 35>, angularLimitsMin=<-45 -10 -35>, angularTarget=<0 0 0>}
      localJointOffset=<0.5 0 0>
     }]
    relativeSpaceBone {boneName=dyn_root_simspace}
    sphericalLimits [
     {limitRadius=12, drivingBone {boneName=None}, sphereLocalOffset=<0 0 0>}
     {limitRadius=12, drivingBone {boneName=None}, sphereLocalOffset=<0 0 0>}
    ]
   }
   n90 = Anim Dynamics - Bone: dyn_hair_bang_b_05!(ComponentPose=@n89, Alpha=0.5, GravityScale:float=1, GravityOverride:Vector="0,0,0", ExternalForce="0,0,0") -> (Pose) {
    angularDampingOverride=0.99
    angularSpringConstant=1500
    bAngularSpring=true
    bGravityOverrideInSimSpace=true
    bOverrideAngularDamping=true
    bUseSphericalLimits=true
    simulationSpace=BoneRelative
    alphaBoolBlend {blendOption=Linear}
    boundBone {boneName=dyn_hair_bang_b_05}
    componentAppliedLinearAccClamp=<5000 5000 5000>
    componentLinearAccScale=<1.2 1.2 1.3>
    componentLinearVelScale=<1.2 1.2 1.3>
    physicsBodyDefinitions [{
      sphereCollisionRadius=0
      boundBone {boneName=dyn_hair_bang_b_05}
      boxExtents=<10 20 20>
      constraintSetup {angularLimitsMax=<5 5 35>, angularLimitsMin=<-45 -10 -35>, angularTarget=<0 0 0>}
      localJointOffset=<0.5 0 0>
     }]
    relativeSpaceBone {boneName=dyn_root_simspace}
    sphericalLimits [
     {limitRadius=12, drivingBone {boneName=None}, sphereLocalOffset=<0 0 0>}
     {limitRadius=12, drivingBone {boneName=None}, sphereLocalOffset=<0 0 0>}
    ]
   }
   n91 = Anim Dynamics - Bone: dyn_hair_bang_b1_1!(ComponentPose=@n90, Alpha=0.5, GravityScale:float=1, GravityOverride:Vector="0,0,0", ExternalForce="0,0,0") -> (Pose) {
    angularDampingOverride=0.999
    angularSpringConstant=1250
    bAngularSpring=true
    bGravityOverrideInSimSpace=true
    bOverrideAngularDamping=true
    bUseSphericalLimits=true
    simulationSpace=BoneRelative
    alphaBoolBlend {blendOption=Linear}
    boundBone {boneName=dyn_hair_bang_b1_1}
    componentAppliedLinearAccClamp=<5000 5000 5000>
    componentLinearAccScale=<1.2 1.2 1.3>
    componentLinearVelScale=<1.2 1.2 1.3>
    physicsBodyDefinitions [{
      sphereCollisionRadius=0
      boundBone {boneName=dyn_hair_bang_b1_1}
      boxExtents=<10 20 20>
      constraintSetup {angularLimitsMax=<5 5 0>, angularLimitsMin=<-45 -45 -40>, angularTarget=<0 0 0>}
      localJointOffset=<0.5 0 0>
     }]
    relativeSpaceBone {boneName=dyn_root_simspace}
    sphericalLimits [
     {limitRadius=12, drivingBone {boneName=None}, sphereLocalOffset=<0 0 0>}
     {limitRadius=12, drivingBone {boneName=None}, sphereLocalOffset=<0 0 0>}
    ]
   }
   n92 = Anim Dynamics - Bone: dyn_hair_bang_b_2_01!(ComponentPose=@n91, Alpha=0.2, GravityScale:float=1, GravityOverride:Vector="0,0,0", ExternalForce="0,0,0") -> (Pose) {
    angularDampingOverride=0.999
    angularSpringConstant=1250
    bAngularSpring=true
    bGravityOverrideInSimSpace=true
    bOverrideAngularDamping=true
    bUseSphericalLimits=true
    simulationSpace=BoneRelative
    alphaBoolBlend {blendOption=Linear}
    boundBone {boneName=dyn_hair_bang_b_2_01}
    componentAppliedLinearAccClamp=<5000 5000 5000>
    componentLinearAccScale=<1.2 1.2 1.3>
    componentLinearVelScale=<1.2 1.2 1.3>
    physicsBodyDefinitions [{
      sphereCollisionRadius=0
      boundBone {boneName=dyn_hair_bang_b_2_01}
      boxExtents=<10 20 20>
      constraintSetup {angularLimitsMax=<5 40 0>, angularLimitsMin=<-45 -45 -40>, angularTarget=<0 0 0>}
      localJointOffset=<2 0 0>
     }]
    relativeSpaceBone {boneName=root}
    sphericalLimits [
     {limitRadius=15, drivingBone {boneName=head}, sphereLocalOffset=<0 0 -50>}
     {limitRadius=12, drivingBone {boneName=clavicle_r}, sphereLocalOffset=<-6 0 0>}
     {limitRadius=6, drivingBone {boneName=dyn_hair_r_main_09}, sphereLocalOffset=<0 0 0>}
    ]
   }
   n93 = Anim Dynamics - Bone: dyn_hair_bang_b_04(ComponentPose=@n92, Alpha=0.2, GravityScale=1, GravityOverride="0,0,0", ExternalForce="0,0,0") {
    angularDampingOverride=0.999
    angularSpringConstant=1250
    bAngularSpring=true
    bGravityOverrideInSimSpace=true
    bOverrideAngularDamping=true
    bUseSphericalLimits=true
    simulationSpace=BoneRelative
    alphaBoolBlend {blendOption=Linear}
    boundBone {boneName=dyn_hair_bang_b_04}
    componentAppliedLinearAccClamp=<5000 5000 5000>
    componentLinearAccScale=<1.2 1.2 1.3>
    componentLinearVelScale=<1.2 1.2 1.3>
    physicsBodyDefinitions [{
      sphereCollisionRadius=0
      boundBone {boneName=dyn_hair_bang_b_04}
      boxExtents=<10 20 20>
      constraintSetup {angularLimitsMax=<5 40 0>, angularLimitsMin=<-45 -45 -40>, angularTarget=<0 0 0>}
      localJointOffset=<2 0 0>
     }]
    relativeSpaceBone {boneName=root}
    sphericalLimits [
     {limitRadius=15, drivingBone {boneName=head}, sphereLocalOffset=<0 0 -4>}
     {limitRadius=12, drivingBone {boneName=clavicle_r}, sphereLocalOffset=<-6 0 0>}
     {limitRadius=6, drivingBone {boneName=dyn_hair_r_main_09}, sphereLocalOffset=<0 0 0>}
    ]
   }
   n94 = Anim Dynamics - Bone: dyn_hair_bang_b_04(ComponentPose=@n93, Alpha=0.2, GravityScale=1, GravityOverride="0,0,0", ExternalForce="0,0,0") {
    angularDampingOverride=0.999
    angularSpringConstant=1250
    bAngularSpring=true
    bGravityOverrideInSimSpace=true
    bOverrideAngularDamping=true
    bUseSphericalLimits=true
    simulationSpace=BoneRelative
    alphaBoolBlend {blendOption=Linear}
    boundBone {boneName=dyn_hair_bang_b_04}
    componentAppliedLinearAccClamp=<5000 5000 5000>
    componentLinearAccScale=<1.2 1.2 1.3>
    componentLinearVelScale=<1.2 1.2 1.3>
    physicsBodyDefinitions [{
      sphereCollisionRadius=0
      boundBone {boneName=dyn_hair_bang_b_04}
      boxExtents=<10 20 20>
      constraintSetup {angularLimitsMax=<5 40 0>, angularLimitsMin=<-45 -45 -40>, angularTarget=<0 0 0>}
      localJointOffset=<2 0 0>
     }]
    relativeSpaceBone {boneName=root}
    sphericalLimits [
     {limitRadius=15, drivingBone {boneName=head}, sphereLocalOffset=<0 0 -4>}
     {limitRadius=12, drivingBone {boneName=clavicle_r}, sphereLocalOffset=<-6 0 0>}
     {limitRadius=6, drivingBone {boneName=dyn_hair_r_main_09}, sphereLocalOffset=<0 0 0>}
    ]
   }
   n95 = "Constraint - dyn_hair_bang_a_01 (P)"(ComponentPose=@n94, Alpha=0.4, ConstraintWeights_0=1) {
    alphaBoolBlend {blendOption=Linear}
    boneToModify {boneName=dyn_hair_bang_a_01}
    constraintSetup [{targetBone {boneName=head}}]
    constraintWeights [1]
   }
   n96 = "Constraint - dyn_hair_bang_a_02 (P)"(ComponentPose=@n95, Alpha=0.3, ConstraintWeights_0=1) {
    alphaBoolBlend {blendOption=Linear}
    boneToModify {boneName=dyn_hair_bang_a_02}
    constraintSetup [{targetBone {boneName=head}}]
    constraintWeights [1]
   }
   n97 = "Constraint - dyn_hair_bang_a_03 (P)"(ComponentPose=@n96, Alpha=0.2, ConstraintWeights_0=1) {
    alphaBoolBlend {blendOption=Linear}
    boneToModify {boneName=dyn_hair_bang_a_03}
    constraintSetup [{targetBone {boneName=head}}]
    constraintWeights [1]
   }
   n98 = "Constraint - dyn_hair_bang_a_04 (P)"(ComponentPose=@n97, Alpha=0.1, ConstraintWeights_0=1) {
    alphaBoolBlend {blendOption=Linear}
    boneToModify {boneName=dyn_hair_bang_a_04}
    constraintSetup [{targetBone {boneName=head}}]
    constraintWeights [1]
   }
   n99 = "Constraint - dyn_hair_l_main (P)"(ComponentPose=@n98, Alpha=0.2, ConstraintWeights_0=1) {
    alphaBoolBlend {blendOption=Linear}
    boneToModify {boneName=dyn_hair_l_main}
    constraintSetup [{targetBone {boneName=head}}]
    constraintWeights [1]
   }
   n100 = "Constraint - dyn_hair_l_main_c_01 (P)"(ComponentPose=@n99, Alpha=0.1, ConstraintWeights_0=1) {
    alphaBoolBlend {blendOption=Linear}
    boneToModify {boneName=dyn_hair_l_main_c_01}
    constraintSetup [{targetBone {boneName=head}}]
    constraintWeights [1]
   }
   n101 = "Constraint - dyn_hair_l_main_b_02 (P)"(ComponentPose=@n100, Alpha=0.3, ConstraintWeights_0=1) {
    alphaBoolBlend {blendOption=Linear}
    boneToModify {boneName=dyn_hair_l_main_b_02}
    constraintSetup [{targetBone {boneName=clavicle_l}}]
    constraintWeights [1]
   }
   n102 = "Constraint - dyn_hair_l_main_b_03 (P)"(ComponentPose=@n101, Alpha=0.2, ConstraintWeights_0=1) {
    alphaBoolBlend {blendOption=Linear}
    boneToModify {boneName=dyn_hair_l_main_b_03}
    constraintSetup [{targetBone {boneName=clavicle_l}}]
    constraintWeights [1]
   }
   n103 = "Constraint - dyn_hair_l_main_b_04 (P)"(ComponentPose=@n102, Alpha=0.3, ConstraintWeights_0=1) {
    alphaBoolBlend {blendOption=Linear}
    boneToModify {boneName=dyn_hair_l_main_b_04}
    constraintSetup [{targetBone {boneName=clavicle_l}}]
    constraintWeights [1]
   }
   n104 = "Constraint - dyn_hair_l_main_b_05 (P)"(ComponentPose=@n103, Alpha=0.4, ConstraintWeights_0=1) {
    alphaBoolBlend {blendOption=Linear}
    boneToModify {boneName=dyn_hair_l_main_b_05}
    constraintSetup [{targetBone {boneName=clavicle_l}}]
    constraintWeights [1]
   }
   n105 = "Constraint - dyn_hair_l_main_f_01 (P)"(ComponentPose=@n104, Alpha=0.5, ConstraintWeights_0=1) {
    alphaBoolBlend {blendOption=Linear}
    boneToModify {boneName=dyn_hair_l_main_f_01}
    constraintSetup [{targetBone {boneName=dyn_hair_l_main_a_02}}]
    constraintWeights [1]
   }
   n106 = "Constraint - dyn_hair_l_main_f_02 (P)"(ComponentPose=@n105, Alpha=0, ConstraintWeights_0=1) {
    alphaBoolBlend {blendOption=Linear}
    boneToModify {boneName=dyn_hair_l_main_f_02}
    constraintSetup [{targetBone {boneName=dyn_hair_l_main_a_02}}]
    constraintWeights [1]
   }
   n107 = "Constraint - dyn_hair_l_main_c_02 (P)"(ComponentPose=@n106, Alpha=0.25, ConstraintWeights_0=1) {
    alphaBoolBlend {blendOption=Linear}
    boneToModify {boneName=dyn_hair_l_main_c_02}
    constraintSetup [{targetBone {boneName=dyn_hair_c_main_02}}]
    constraintWeights [1]
   }
   n108 = "Constraint - dyn_hair_r_main_02 (R)"(ComponentPose=@n107, Alpha=1, ConstraintWeights_0=1) {
    alphaBoolBlend {blendOption=Linear}
    boneToModify {boneName=dyn_hair_r_main_02}
    constraintSetup [{transformType=Rotation, targetBone {boneName=clavicle_r}}]
    constraintWeights [1]
   }
   n109 = "Constraint - dyn_hair_r_main_03 (PP)"(ComponentPose=@n108, Alpha=1, ConstraintWeights_0=0.5, ConstraintWeights_1=0.5) {
    alphaBoolBlend {blendOption=Linear}
    boneToModify {boneName=dyn_hair_r_main_03}
    constraintSetup [{targetBone {boneName=dyn_hair_c_main_02}}, {targetBone {boneName=spine_05}}]
    constraintWeights [1,1]
   }
   n110 = "Constraint - dyn_hair_r_main_09 (PP)"!(ComponentPose=@n109, Alpha=0.7, ConstraintWeights_0=0.5, ConstraintWeights_1:float=0.5) -> (Pose) {
    alphaBoolBlend {blendOption=Linear}
    boneToModify {boneName=dyn_hair_r_main_09}
    constraintSetup [{targetBone {boneName=dyn_hair_r_main_03}}, {targetBone {boneName=clavicle_r}}]
    constraintWeights [1,1]
   }
   n111 = AND Boolean(A=Get bIsSkyDiving(), B=NOT Boolean(A=Get bIsParachuteOpened()))
   n112 = "Transform (Modify) Bone - Bone: dyn_hair_l_main_c_02"!(ComponentPose=@n110, bAlphaBoolEnabled:bool=@n111, Translation="0,0,0", Rotation="20,0.000,0.000", Scale="1,1,1") -> (Pose) {
    alphaInputType=Bool
    rotationMode=BMM_Additive
    rotationSpace=BCS_BoneSpace
    scaleMode=BMM_Additive
    scaleSpace=BCS_BoneSpace
    translationMode=BMM_Additive
    translationSpace=BCS_BoneSpace
    alphaBoolBlend {blendOption=Linear}
    boneToModify {boneName=dyn_hair_l_main_c_02}
    rotation {pitch=20}
   }
   n113 = "Transform (Modify) Bone - Bone: dyn_hair_c_main_02"!(ComponentPose=@n112, bAlphaBoolEnabled:bool=@n111, Translation="0,0,0", Rotation="22,0,0", Scale="1,1,1") -> (Pose) {
    alphaInputType=Bool
    rotationMode=BMM_Additive
    rotationSpace=BCS_BoneSpace
    scaleMode=BMM_Additive
    scaleSpace=BCS_BoneSpace
    translationMode=BMM_Additive
    translationSpace=BCS_BoneSpace
    alphaBoolBlend {blendOption=Linear}
    boneToModify {boneName=dyn_hair_c_main_02}
    rotation {pitch=22}
   }
   n114 = "Transform (Modify) Bone - Bone: dyn_hair_r_main_03"!(ComponentPose=@n113, bAlphaBoolEnabled:bool=@n111, Translation="0,0,0", Rotation="20,0,0", Scale="1,1,1") -> (Pose) {
    alphaInputType=Bool
    rotationMode=BMM_Additive
    rotationSpace=BCS_BoneSpace
    scaleMode=BMM_Additive
    scaleSpace=BCS_BoneSpace
    translationMode=BMM_Additive
    translationSpace=BCS_BoneSpace
    alphaBoolBlend {blendOption=Linear}
    boneToModify {boneName=dyn_hair_r_main_03}
    rotation {pitch=20}
   }
   n115 = "Transform (Modify) Bone - Bone: dyn_hair_l_main_c_03"!(ComponentPose=@n114, bAlphaBoolEnabled:bool=@n111, Translation="0,0,0", Rotation="20,0.000,0.000", Scale="1,1,1") -> (Pose) {
    alphaInputType=Bool
    rotationMode=BMM_Additive
    rotationSpace=BCS_BoneSpace
    scaleMode=BMM_Additive
    scaleSpace=BCS_BoneSpace
    translationMode=BMM_Additive
    translationSpace=BCS_BoneSpace
    alphaBoolBlend {blendOption=Linear}
    boneToModify {boneName=dyn_hair_l_main_c_03}
    rotation {pitch=20}
   }
   n116 = "Transform (Modify) Bone - Bone: dyn_hair_c_main_03"!(ComponentPose=@n115, bAlphaBoolEnabled:bool=@n111, Translation="0,0,0", Rotation="22,0,0", Scale="1,1,1") -> (Pose) {
    alphaInputType=Bool
    rotationMode=BMM_Additive
    rotationSpace=BCS_BoneSpace
    scaleMode=BMM_Additive
    scaleSpace=BCS_BoneSpace
    translationMode=BMM_Additive
    translationSpace=BCS_BoneSpace
    alphaBoolBlend {blendOption=Linear}
    boneToModify {boneName=dyn_hair_c_main_03}
    rotation {pitch=22}
   }
   n117 = "Transform (Modify) Bone - Bone: dyn_hair_r_main_04"!(ComponentPose=@n116, bAlphaBoolEnabled:bool=@n111, Translation="0,0,0", Rotation="20,0,0", Scale="1,1,1") -> (Pose) {
    alphaInputType=Bool
    rotationMode=BMM_Additive
    rotationSpace=BCS_BoneSpace
    scaleMode=BMM_Additive
    scaleSpace=BCS_BoneSpace
    translationMode=BMM_Additive
    translationSpace=BCS_BoneSpace
    alphaBoolBlend {blendOption=Linear}
    boneToModify {boneName=dyn_hair_r_main_04}
    rotation {pitch=20}
   }
   n118 = Get Crouching() -> (Crouching:double)
   n119 = Anim Dynamics - Bone: dyn_hair_l_main_c_03!(ComponentPose=@n117, Alpha=@n118, ExternalForce="0,0,0") -> (Pose) {
    angularDampingOverride=1
    bGravityOverrideInSimSpace=true
    bOverrideAngularDamping=true
    simulationSpace=RootRelative
    alphaBoolBlend {blendOption=Linear}
    boundBone {boneName=dyn_hair_l_main_c_03}
    physicsBodyDefinitions [{
      sphereCollisionRadius=0
      boundBone {boneName=dyn_hair_l_main_c_03}
      boxExtents=<10 10 10>
      constraintSetup {angularLimitsMax=<2 -2 45>, angularLimitsMin=<-2 -45 -45>, angularTarget=<0 0 0>}
      localJointOffset=<1.5 -0.05 0.1>
     }]
   }
   n120 = Anim Dynamics - Bone: dyn_hair_l_main_c_04!(ComponentPose=@n119, Alpha=@n118, ExternalForce="0,0,0") -> (Pose) {
    angularDampingOverride=1
    bGravityOverrideInSimSpace=true
    bOverrideAngularDamping=true
    simulationSpace=RootRelative
    alphaBoolBlend {blendOption=Linear}
    boundBone {boneName=dyn_hair_l_main_c_04}
    physicsBodyDefinitions [{
      sphereCollisionRadius=0
      boundBone {boneName=dyn_hair_l_main_c_04}
      boxExtents=<10 10 10>
      constraintSetup {angularLimitsMax=<2 -2 45>, angularLimitsMin=<-2 -45 -45>, angularTarget=<0 0 0>}
      localJointOffset=<1.5 -0.05 0.1>
     }]
   }
   n121 = Anim Dynamics - Bone: dyn_hair_l_main_c_04(ComponentPose=@n120, Alpha=@n118, ExternalForce="0,0,0") {
    angularDampingOverride=1
    bGravityOverrideInSimSpace=true
    bOverrideAngularDamping=true
    simulationSpace=RootRelative
    alphaBoolBlend {blendOption=Linear}
    boundBone {boneName=dyn_hair_l_main_c_04}
    physicsBodyDefinitions [{
      sphereCollisionRadius=0
      boundBone {boneName=dyn_hair_l_main_c_04}
      boxExtents=<10 10 10>
      constraintSetup {angularLimitsMax=<2 -2 45>, angularLimitsMin=<-2 -45 -45>, angularTarget=<0 0 0>}
      localJointOffset=<1.5 -0.05 0.1>
     }]
   }
   n122 = Anim Dynamics - Bone: dyn_hair_l_main_c_05!(ComponentPose=@n121, Alpha=@n118, ExternalForce="0,0,0") -> (Pose) {
    angularDampingOverride=1
    bGravityOverrideInSimSpace=true
    bOverrideAngularDamping=true
    simulationSpace=RootRelative
    alphaBoolBlend {blendOption=Linear}
    boundBone {boneName=dyn_hair_l_main_c_05}
    physicsBodyDefinitions [{
      sphereCollisionRadius=0
      boundBone {boneName=dyn_hair_l_main_c_05}
      boxExtents=<10 10 10>
      constraintSetup {angularLimitsMax=<2 -2 45>, angularLimitsMin=<-2 -45 -45>, angularTarget=<0 0 0>}
      localJointOffset=<1.5 -0.05 0.1>
     }]
   }
   n123 = Anim Dynamics - Bone: dyn_hair_c_main_03!(ComponentPose=@n122, Alpha=@n118, ExternalForce="0,0,0") -> (Pose) {
    angularDampingOverride=1
    bGravityOverrideInSimSpace=true
    bOverrideAngularDamping=true
    simulationSpace=RootRelative
    alphaBoolBlend {blendOption=Linear}
    boundBone {boneName=dyn_hair_c_main_03}
    physicsBodyDefinitions [{
      sphereCollisionRadius=0
      boundBone {boneName=dyn_hair_c_main_03}
      boxExtents=<12 12 12>
      constraintSetup {angularLimitsMax=<2 -2 45>, angularLimitsMin=<-2 -45 -45>, angularTarget=<0 0 0>}
      localJointOffset=<1.5 0 0.1>
     }]
   }
   n124 = Anim Dynamics - Bone: dyn_hair_c_main_04!(ComponentPose=@n123, Alpha=@n118, ExternalForce="0,0,0") -> (Pose) {
    angularDampingOverride=1
    bGravityOverrideInSimSpace=true
    bOverrideAngularDamping=true
    simulationSpace=RootRelative
    alphaBoolBlend {blendOption=Linear}
    boundBone {boneName=dyn_hair_c_main_04}
    physicsBodyDefinitions [{
      sphereCollisionRadius=0
      boundBone {boneName=dyn_hair_c_main_04}
      boxExtents=<12 12 12>
      constraintSetup {angularLimitsMax=<2 -2 45>, angularLimitsMin=<-2 -45 -45>, angularTarget=<0 0 0>}
      localJointOffset=<1.5 0 0.1>
     }]
   }
   n125 = Anim Dynamics - Bone: dyn_hair_c_main_05!(ComponentPose=@n124, Alpha=@n118, ExternalForce="0,0,0") -> (Pose) {
    angularDampingOverride=1
    bGravityOverrideInSimSpace=true
    bOverrideAngularDamping=true
    simulationSpace=RootRelative
    alphaBoolBlend {blendOption=Linear}
    boundBone {boneName=dyn_hair_c_main_05}
    physicsBodyDefinitions [{
      sphereCollisionRadius=0
      boundBone {boneName=dyn_hair_c_main_05}
      boxExtents=<12 12 12>
      constraintSetup {angularLimitsMax=<2 -2 45>, angularLimitsMin=<-2 -45 -45>, angularTarget=<0 0 0>}
      localJointOffset=<1.5 0 0.1>
     }]
   }
   n126 = Anim Dynamics - Bone: dyn_hair_r_main_04!(ComponentPose=@n125, Alpha=@n118, ExternalForce="0,0,0") -> (Pose) {
    angularDampingOverride=1
    bGravityOverrideInSimSpace=true
    bOverrideAngularDamping=true
    simulationSpace=RootRelative
    alphaBoolBlend {blendOption=Linear}
    boundBone {boneName=dyn_hair_r_main_04}
    physicsBodyDefinitions [{
      sphereCollisionRadius=0
      boundBone {boneName=dyn_hair_r_main_04}
      boxExtents=<11 11 11>
      constraintSetup {angularLimitsMax=<2 -2 45>, angularLimitsMin=<-2 -45 -45>, angularTarget=<0 0 0>}
      localJointOffset=<1.5 -0.05 -0.5>
     }]
   }
   n127 = Anim Dynamics - Bone: dyn_hair_r_main_05!(ComponentPose=@n126, Alpha=@n118, ExternalForce="0,0,0") -> (Pose) {
    angularDampingOverride=1
    bGravityOverrideInSimSpace=true
    bOverrideAngularDamping=true
    simulationSpace=RootRelative
    alphaBoolBlend {blendOption=Linear}
    boundBone {boneName=dyn_hair_r_main_05}
    physicsBodyDefinitions [{
      sphereCollisionRadius=0
      boundBone {boneName=dyn_hair_r_main_05}
      boxExtents=<11 11 11>
      constraintSetup {angularLimitsMax=<2 -2 45>, angularLimitsMin=<-2 -45 -45>, angularTarget=<0 0 0>}
      localJointOffset=<1.5 -0.05 -0.5>
     }]
   }
   n128 = Anim Dynamics - Bone: dyn_hair_r_main_06!(ComponentPose=@n127, Alpha=@n118, ExternalForce="0,0,0") -> (Pose) {
    angularDampingOverride=1
    bGravityOverrideInSimSpace=true
    bOverrideAngularDamping=true
    simulationSpace=RootRelative
    alphaBoolBlend {blendOption=Linear}
    boundBone {boneName=dyn_hair_r_main_06}
    physicsBodyDefinitions [{
      sphereCollisionRadius=0
      boundBone {boneName=dyn_hair_r_main_06}
      boxExtents=<11 11 11>
      constraintSetup {angularLimitsMax=<2 -2 45>, angularLimitsMin=<-2 -45 -45>, angularTarget=<0 0 0>}
      localJointOffset=<1.5 -0.05 -0.5>
     }]
   }
   n129 = Anim Dynamics - Bone: dyn_hair_r_main_07!(ComponentPose=@n128, Alpha=@n118, ExternalForce="0,0,0") -> (Pose) {
    angularDampingOverride=1
    bGravityOverrideInSimSpace=true
    bOverrideAngularDamping=true
    simulationSpace=RootRelative
    alphaBoolBlend {blendOption=Linear}
    boundBone {boneName=dyn_hair_r_main_07}
    physicsBodyDefinitions [{
      sphereCollisionRadius=0
      boundBone {boneName=dyn_hair_r_main_07}
      boxExtents=<11 11 11>
      constraintSetup {angularLimitsMax=<2 -2 45>, angularLimitsMin=<-2 -45 -45>, angularTarget=<0 0 0>}
      localJointOffset=<1.5 -0.05 -0.5>
     }]
   }
   n130 = "Constraint - dyn_hair_r_main_04 (PPP)"!(ComponentPose=@n129, Alpha=1, ConstraintWeights_0=0, ConstraintWeights_1:float=0, ConstraintWeights_2:float=1) -> (Pose) {
    alphaBoolBlend {blendOption=Linear}
    boneToModify {boneName=dyn_hair_r_main_04}
    constraintSetup [{targetBone {boneName=spine_05}}, {targetBone {boneName=dyn_hair_c_main_02}}, {targetBone {boneName=dyn_hair_c_main_03}}]
    constraintWeights [1,1,1]
   }
   n131 = "Transform (Modify) Bone - Bone: dyn_hair_r_main_03"!(ComponentPose=@n130, Alpha:float=1, Translation="0,0,3", Rotation="0,-10,0", Scale="1,1,1") -> (Pose) {
    rotationMode=BMM_Additive
    rotationSpace=BCS_BoneSpace
    scaleMode=BMM_Additive
    scaleSpace=BCS_BoneSpace
    translationMode=BMM_Additive
    translationSpace=BCS_BoneSpace
    alphaBoolBlend {blendOption=Linear}
    alphaScaleBiasClamp {bInterpResult=true, interpSpeedDecreasing=20, interpSpeedIncreasing=20}
    boneToModify {boneName=dyn_hair_r_main_03}
    rotation {yaw=-10}
    translation {z=3}
   }
   n132 = "Transform (Modify) Bone - Bone: dyn_hair_bang_b_2_01"(ComponentPose=@n131, Alpha=1, Translation="0,0,0", Rotation="0,0,0", Scale="0.8,0.8,0.8") {
    rotationMode=BMM_Additive
    rotationSpace=BCS_BoneSpace
    scaleMode=BMM_Additive
    scaleSpace=BCS_BoneSpace
    translationMode=BMM_Additive
    translationSpace=BCS_BoneSpace
    alphaBoolBlend {blendOption=Linear}
    alphaScaleBiasClamp {bInterpResult=true, interpSpeedDecreasing=20, interpSpeedIncreasing=20}
    boneToModify {boneName=dyn_hair_bang_b_2_01}
    scale=<0.8 0.8 0.8>
   }
   Save cached pose 'PostSim'(Pose=Component To Local(ComponentPose=@n132)) {cacheName=PostSim}
  ]}
 EventGraph {nodes [
   Event BlueprintUpdateAnimation()
   > n1 = Sequence!(execute) -> (then_0, then_1, then_2:Exec, then_3:Exec, then_4:Exec)
   n2 = ComplexSimSpaceFunction DEPRECATED!(
    execute=@n1.then_1
    CharacterPartAnimInstance:Custom Character Part Anim Instance Object Reference=Self-Reference()
    InterpSpeedMax:double=9
    PawnSpeedMax:double=20
    RInterpErrorTolerance:double=1
    DeltaTime:double=0
    DebugDraw:bool=false
    sim_space_rotation_in:Rotator=Get sim_space_rotation()
    inComponentSpaceIn:bool=Get InComponentSpace() -> (InComponentSpace:bool)
    inComponentSpaceLastFrameIn:bool=Get InComponentSpaceLastFrame() -> (InComponentSpaceLastFrame:bool)
    lockedOnSimSpaceIn:bool=Get LockedOnSimSpace() -> (LockedOnSimSpace:bool)
    InterpSpeedIn:double=Get InterpSpeed() -> (InterpSpeed:double)
   ) -> (
    then
    sim_space_rotation:Rotator
    inComponentSpaceOut:bool
    inComponentSpaceLastFrameOut:bool
    lockedOnSimSpaceOut:bool
    InterpSpeedOut:double
   )
   > Set sim_space_rotation(sim_space_rotation:Rotator=@n2) -> (Output_Get:Rotator)
   > Set InComponentSpace(InComponentSpace:bool=@n2.inComponentSpaceOut) -> (Output_Get:bool)
   > Set InComponentSpaceLastFrame(InComponentSpaceLastFrame:bool=@n2.inComponentSpaceLastFrameOut) -> (Output_Get:bool)
   > Set LockedOnSimSpace(LockedOnSimSpace:bool=@n2.lockedOnSimSpaceOut) -> (Output_Get:bool)
   > Set InterpSpeed(InterpSpeed:double=@n2.InterpSpeedOut) -> (Output_Get:double)
   n3 = Gravity Override State Function DEPRECATED(
    execute=@n1.then_0
    CharacterPartAnimInstance=Self-Reference()
    GameTimeInSeconds=GetGameTimeInSeconds()
    GravityOverrideParams=Get Gravity Override Params()
   )
   > Set GravityOverride(GravityOverride=@n3)
   n4 = GetOwnerSkeletalMeshComponent()
   n5 = BreakRotator(InRot=BreakTransform(InTransform=GetDeltaTransformFromRefPose(self=@n4, BoneName=head, BaseName=neck_02)).Rotation)
   Set LookUp(execute=@n1.then_2, LookUp:double=MapRangeClamped(Value=@n5.Yaw, InRangeA=0, InRangeB=-20)) -> (Output_Get:double)
   n6 = BreakTransform(InTransform=GetDeltaTransformFromRefPose(self=@n4, BoneName=head, BaseName=neck_02))
   > n7 = Set RifleLookUp(RifleLookUp:double=MapRangeClamped(Value=BreakRotator(InRot=@n6.Rotation).Yaw, InRangeA=5, InRangeB=-20)) -> (Output_Get:double)
   n8 = CheckForBackbling()
   Set isNotWearingBackbling(execute=[@n7.then, @n12.then], isNotWearingBackbling:double="To Float (Boolean)"(InBool=NOT Boolean(A=@n8))) -> (Output_Get:double)
   > Set isWearingCape(isWearingCape=@n8.bIsBackblingLong)
   > Set CrouchWalking(CrouchWalking:bool=AND Boolean(A=Get bIsCrouching(), B=Get bIsAccelerating2D())) -> (Output_Get:bool)
   > Set Crouching(Crouching:double=SelectFloat(A=0.25, B=1, bPickA=Get bIsCrouching())) -> (Output_Get:double)
   n9 = Cast To FortPlayerAnimInstance(execute=@n1.then_3, Object=GetAnimInstance(self=GetOwnerSkeletalMeshComponent()))
   > Set isAimingPistol(isAimingPistol:double=0) -> (Output_Get:double)
   > Set isAiming Rifle(isAiming Rifle:double=0) -> (Output_Get:double)
   > n10 = Switch on EFortWeaponCoreAnimation!(execute, Selection:EFortWeaponCoreAnimation Enum=Get WeaponCoreAnim(self=@n9)) -> (
    Melee:Exec
    Pistol:Exec
    Shotgun:Exec
    PaperBlueprint:Exec
    Rifle:Exec
    MeleeOneHand:Exec
    MachinePistol:Exec
    RocketLauncher:Exec
    GrenadeLauncher:Exec
    GoingCommando:Exec
    AssaultRifle:Exec
    TacticalShotgun:Exec
    SniperRifle:Exec
    TrapPlacement:Exec
    ShoulderLauncher:Exec
    AbilityDecoTool:Exec
    Crossbow:Exec
    C4:Exec
    RemoteControl:Exec
    DualWield:Exec
    AR_BullPup:Exec
    AR_ForwardGrip:Exec
    MedPackPaddles:Exec
    SMG_P90:Exec
    AR_DrumGun:Exec
    Consumable_Small:Exec
    Consumable_Large:Exec
    Balloon:Exec
    MountedTurret:Exec
    CreativeTool:Exec
    ExplosiveBow:Exec
    AshtonIndigo:Exec
    AshtonChicago:Exec
    MeleeDualWield:Exec
    Unarmed:Exec
   )
   Branch(execute=[
     @n10.AssaultRifle
     @n10.Crossbow
     @n10.GrenadeLauncher
     @n10.Rifle
     @n10.RocketLauncher
     @n10.Shotgun
     @n10.SniperRifle
     @n10.TacticalShotgun
    ])
   > Set isAiming Rifle(isAiming Rifle=1)
   Branch(execute=[@n10.MachinePistol, @n10.Pistol])
   > Set isAimingPistol(isAimingPistol=1)
   Event OnCharacterCustomizationLoaded()
   n11 = NOT Boolean(A=@n8.bIsBackblingLong)
   > n12 = Set isWearingBackpack(isWearingBackpack:double="To Float (Boolean)"(InBool=AND Boolean(A=@n8, B=@n11))) -> (Output_Get:double)
   "Absolute (Float)"(A=@n5.Yaw)
  ]}
]
variables [
 Crouching {type=number}
 CrouchWalking {type=boolean}
 Gravity Override Params {type=GravityOverrideParamsStruct}
 GravityOverride {type=Vector}
 InComponentSpace {type=boolean}
 InComponentSpaceLastFrame {type=boolean}
 InterpSpeed {type=number}
 isAiming Rifle {type=number}
 isAimingPistol {type=number}
 isNotWearingBackbling {type=number}
 isWearingBackpack {type=number}
 isWearingCape {type=boolean}
 LockedOnSimSpace {type=boolean}
 LookUp {type=number}
 RifleLookUp {type=number}
 sim_space_rotation {type=Rotator}
 Skydiving {type=number}
 windVector {type=Vector}
]
